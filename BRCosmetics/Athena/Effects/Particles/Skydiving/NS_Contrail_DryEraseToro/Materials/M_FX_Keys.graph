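# Translucent unlit material for Niagara mesh particle keys in the DryEraseToro skydiving contrail.
# A per-particle KeyIndex offsets two noise lookups so keys vary: panning noise distorts UVs, while blurred noise varies edge falloff.
# An inverted sphere mask on distorted U raises opacity from BaseOpacity toward the edges, with vertex red adding more and particle alpha scaling it.
# Particle color, boosted by vertex green, feeds the shared emissive adjustment function.
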Material /BRCosmetics/Athena/Effects/Particles/Skydiving/NS_Contrail_DryEraseToro/Materials/M_FX_Keys
graph {
 expressions {
  MaterialExpressionAdd_1 {
   class=MaterialExpressionAdd
   inputs [
    A {from=MaterialExpressionPower_0}
    B {from=MaterialExpressionMultiply_1}
   ]
   properties {constB=0}
  }
  MaterialExpressionAdd_2 {
   class=MaterialExpressionAdd
   inputs [
    A {from=MaterialExpressionMaterialFunctionCall_0, output=G}
    B {from=MaterialExpressionAdd_6}
   ]
  }
  MaterialExpressionAdd_4 {
   class=MaterialExpressionAdd
   inputs [
    A {from=MaterialExpressionMaterialFunctionCall_5, output=G}
    B {from=MaterialExpressionDynamicParameter_1, output=KeyIndex}
   ]
  }
  MaterialExpressionAdd_5 {
   class=MaterialExpressionAdd
   inputs [
    A {from=MaterialExpressionTextureCoordinate_0}
    B {from=MaterialExpressionMultiply_7}
   ]
  }
  MaterialExpressionAdd_6 {
   class=MaterialExpressionAdd
   inputs [
    A {from=MaterialExpressionDynamicParameter_0, output=KeyIndex}
    B {from=""}
   ]
   properties {constB=0.1}
  }
  MaterialExpressionAppendVector_0 {
   class=MaterialExpressionAppendVector
   inputs [
    A {from=MaterialExpressionMaterialFunctionCall_0, output=R}
    B {from=MaterialExpressionAdd_2, output=""}
   ]
  }
  MaterialExpressionAppendVector_1 {
   class=MaterialExpressionAppendVector
   inputs [
    A {from=MaterialExpressionMaterialFunctionCall_5, output=R}
    B {from=MaterialExpressionAdd_4, output=""}
   ]
  }
  MaterialExpressionConstantBiasScale_0 {class=MaterialExpressionConstantBiasScale, inputs [None {from=MaterialExpressionTextureSample_1, output=R}], properties {bias=-0.5, scale=2}}
  MaterialExpressionConstant_0 {class=MaterialExpressionConstant, properties {r=0.5}}
  MaterialExpressionConstant_1 {class=MaterialExpressionConstant, properties {r=0.8}}
  MaterialExpressionConstant_10 {class=MaterialExpressionConstant, properties {r=0.3}}
  MaterialExpressionConstant_11 {class=MaterialExpressionConstant, properties {r=5}}
  MaterialExpressionConstant_12 {class=MaterialExpressionConstant, properties {r=2.5}}
  MaterialExpressionConstant_2 {class=MaterialExpressionConstant, properties {r=2}}
  MaterialExpressionConstant_3 {class=MaterialExpressionConstant, properties {r=-0.2}}
  MaterialExpressionConstant_4 {class=MaterialExpressionConstant}
  MaterialExpressionConstant_5 {class=MaterialExpressionConstant, properties {r=2}}
  MaterialExpressionConstant_6 {class=MaterialExpressionConstant, properties {r=0.5}}
  MaterialExpressionConstant_7 {class=MaterialExpressionConstant, properties {r=2}}
  MaterialExpressionConstant_8 {class=MaterialExpressionConstant, properties {r=4}}
  MaterialExpressionConstant_9 {class=MaterialExpressionConstant, properties {r=2}}
  MaterialExpressionConvert_0 {
   class=MaterialExpressionConvert
   inputs [
    None {from=""}
    None {from=MaterialExpressionConstant_3}
   ]
   properties {
    nodeName=Make Float2
    convertInputs [
     {type=Scalar, expressionInput {expression=null, inputName=None, mask=0, maskA=0, maskB=0, maskG=0, maskR=0, outputIndex=0}}
     {
      type=Scalar
      expressionInput {
       expression=&self:MaterialExpressionConstant_3
       inputName=None
       mask=0
       maskA=0
       maskB=0
       maskG=0
       maskR=0
       outputIndex=0
      }
     }
    ]
    convertOutputs [{type=Vector2}]
   }
  }
  MaterialExpressionDynamicParameter_0 {class=MaterialExpressionDynamicParameter, inputs [], properties {paramNames [KeyIndex,Param2,Param3,Param4]}}
  MaterialExpressionDynamicParameter_1 {class=MaterialExpressionDynamicParameter, inputs [], properties {paramNames [KeyIndex,Param2,Param3,Param4]}}
  MaterialExpressionLinearInterpolate_0 {
   class=MaterialExpressionLinearInterpolate
   inputs [
    A {from=MaterialExpressionScalarParameter_1}
    B {from=""}
    Alpha {from=MaterialExpressionPower_2}
   ]
  }
  MaterialExpressionLinearInterpolate_1 {
   class=MaterialExpressionLinearInterpolate
   inputs [
    A {from=""}
    B {from=MaterialExpressionConstant_11}
    Alpha {from=MaterialExpressionPower_1}
   ]
   properties {constA=1}
  }
  MaterialExpressionLinearInterpolate_2 {
   class=MaterialExpressionLinearInterpolate
   inputs [
    A {from=MaterialExpressionConstant_7}
    B {from=MaterialExpressionConstant_8}
    Alpha {from=MaterialExpressionMultiply_5}
   ]
  }
  MaterialExpressionMaterialFunctionCall_0 {
   class=MaterialExpressionMaterialFunctionCall
   inputs [Float2 {from=MaterialExpressionMultiply_4}]
   properties {materialFunction=&BreakOutFloat2Components}
  }
  MaterialExpressionMaterialFunctionCall_1 {
   class=MaterialExpressionMaterialFunctionCall
   inputs [Float2 {from=MaterialExpressionAdd_5}]
   properties {materialFunction=&BreakOutFloat2Components}
  }
  MaterialExpressionMaterialFunctionCall_2 {
   class=MaterialExpressionMaterialFunctionCall
   inputs [Emissive {from=MaterialExpressionMultiply_3}]
   properties {materialFunction=&MF_FX_EmissiveAdjustments}
  }
  MaterialExpressionMaterialFunctionCall_3 {
   class=MaterialExpressionMaterialFunctionCall
   inputs [
    X {from=MaterialExpressionConstant_4}
    Y {from=MaterialExpressionConstant_5}
   ]
   properties {materialFunction=&MakeFloat2}
  }
  MaterialExpressionMaterialFunctionCall_4 {
   class=MaterialExpressionMaterialFunctionCall
   inputs [
    In {from=MaterialExpressionTextureSample_0, output=R}
    Contrast {from=MaterialExpressionConstant_9}
   ]
   properties {materialFunction=&CheapContrast}
  }
  MaterialExpressionMaterialFunctionCall_5 {
   class=MaterialExpressionMaterialFunctionCall
   inputs [Float2 {from=MaterialExpressionTextureCoordinate_2}]
   properties {materialFunction=&BreakOutFloat2Components}
  }
  MaterialExpressionMaterialFunctionCall_6 {
   class=MaterialExpressionMaterialFunctionCall
   inputs [
    X {from=MaterialExpressionConstant_1}
    Y {from=MaterialExpressionConstant_2}
   ]
   properties {materialFunction=&MakeFloat2}
  }
  MaterialExpressionMultiply_0 {
   class=MaterialExpressionMultiply
   inputs [
    A {from=MaterialExpressionParticleColor_0, output=A}
    B {from=MaterialExpressionSaturate_0}
   ]
  }
  MaterialExpressionMultiply_1 {
   class=MaterialExpressionMultiply
   inputs [
    A {from=MaterialExpressionVertexColor_0, output=R}
    B {from=MaterialExpressionConstant_10}
   ]
  }
  MaterialExpressionMultiply_3 {
   class=MaterialExpressionMultiply
   inputs [
    A {from=MaterialExpressionParticleColor_0, output=RGB}
    B {from=MaterialExpressionLinearInterpolate_1}
   ]
  }
  MaterialExpressionMultiply_4 {
   class=MaterialExpressionMultiply
   inputs [
    A {from=MaterialExpressionTextureCoordinate_1}
    B {from=MaterialExpressionMaterialFunctionCall_3, output=Result}
   ]
  }
  MaterialExpressionMultiply_5 {
   class=MaterialExpressionMultiply
   inputs [
    A {from=MaterialExpressionMaterialFunctionCall_4, output=Result}
    B {from=""}
   ]
  }
  MaterialExpressionMultiply_7 {
   class=MaterialExpressionMultiply
   inputs [
    A {from=MaterialExpressionConstantBiasScale_0}
    B {from=MaterialExpressionScalarParameter_16}
   ]
  }
  MaterialExpressionMultiply_8 {
   class=MaterialExpressionMultiply
   inputs [
    A {from=MaterialExpressionAppendVector_1}
    B {from=MaterialExpressionMaterialFunctionCall_6, output=Result}
   ]
  }
  MaterialExpressionOneMinus_0 {class=MaterialExpressionOneMinus, inputs [None {from=MaterialExpressionSphereMask_0}]}
  MaterialExpressionPanner_0 {
   class=MaterialExpressionPanner
   inputs [
    Coordinate {from=MaterialExpressionMultiply_8}
    Time {from=""}
    Speed {from=MaterialExpressionConvert_0}
   ]
  }
  MaterialExpressionParticleColor_0 {class=MaterialExpressionParticleColor, inputs [], properties {}}
  MaterialExpressionPower_0 {
   class=MaterialExpressionPower
   inputs [
    Base {from=MaterialExpressionLinearInterpolate_0}
    Exp {from=""}
   ]
  }
  MaterialExpressionPower_1 {
   class=MaterialExpressionPower
   inputs [
    Base {from=MaterialExpressionVertexColor_1, output=G}
    Exp {from=MaterialExpressionConstant_12}
   ]
  }
  MaterialExpressionPower_2 {
   class=MaterialExpressionPower
   inputs [
    Base {from=MaterialExpressionOneMinus_0}
    Exp {from=MaterialExpressionLinearInterpolate_2}
   ]
  }
  MaterialExpressionSaturate_0 {class=MaterialExpressionSaturate, inputs [None {from=MaterialExpressionAdd_1}]}
  MaterialExpressionScalarParameter_1 {class=MaterialExpressionScalarParameter, properties {defaultValue=0.3, parameterName=BaseOpacity}}
  MaterialExpressionScalarParameter_16 {class=MaterialExpressionScalarParameter, properties {defaultValue=0.1, parameterName=Distorion}}
  MaterialExpressionSphereMask_0 {
   class=MaterialExpressionSphereMask
   inputs [
    A {from=MaterialExpressionMaterialFunctionCall_1, output=R}
    B {from=MaterialExpressionConstant_0, output=""}
    Radius {from=MaterialExpressionConstant_6, output=""}
    Hardness {from="", output=""}
   ]
   properties {hardnessPercent=0}
  }
  MaterialExpressionTextureCoordinate_0 {class=MaterialExpressionTextureCoordinate}
  MaterialExpressionTextureCoordinate_1 {class=MaterialExpressionTextureCoordinate, properties {vTiling=0.0455}}
  MaterialExpressionTextureCoordinate_2 {class=MaterialExpressionTextureCoordinate, properties {vTiling=0.0455}}
  MaterialExpressionTextureSample_0 {
   class=MaterialExpressionTextureSample
   inputs [
    UVs {from=MaterialExpressionAppendVector_0}
    Tex {from=""}
    Apply View MipBias {from=""}
   ]
   properties {samplerType=SAMPLERTYPE_Grayscale, texture=&LowResBlurredNoise}
  }
  MaterialExpressionTextureSample_1 {
   class=MaterialExpressionTextureSample
   inputs [
    UVs {from=MaterialExpressionPanner_0}
    Tex {from=""}
    Apply View MipBias {from=""}
   ]
   properties {samplerType=SAMPLERTYPE_LinearColor, texture=&HighContrastNoise}
  }
  MaterialExpressionVertexColor_0 {class=MaterialExpressionVertexColor, inputs [], properties {}}
  MaterialExpressionVertexColor_1 {class=MaterialExpressionVertexColor, inputs [], properties {}}
 }
 properties {
  bAutomaticallySetUsageInEditor=false
  bUsedWithNiagaraMeshParticles=true
  blendMode=BLEND_Translucent
  shadingModel=MSM_Unlit
  usedShadingModels=Unlit
 }
}
