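# Additive, unlit, two-sided glow material for Niagara mesh particles, named for the keys effect in the DryEraseToro skydiving contrail.
# Particle color feeds an emissive adjustment function.
# A separate mask multiplies an inverted V gradient, shaped by dynamic GlowPower, with particle alpha.
# That mask also takes an optional vertex-red vertical falloff and a view-angle fade to soften the glow.
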Material /BRCosmetics/Athena/Effects/Particles/Skydiving/NS_Contrail_DryEraseToro/Materials/M_FX_Keys_Glow
graph {
 properties {
  bAutomaticallySetUsageInEditor=false
  bUsedWithNiagaraMeshParticles=true
  blendMode=BLEND_Additive
  shadingModel=MSM_Unlit
  twoSided=true
  usedShadingModels=Unlit
 }
 nodes [
  n1 = ParticleColor()
  MaterialFunctionCall(Emissive=@n1.RGB) {materialFunction=&MF_FX_EmissiveAdjustments}
  n2 = TextureCoordinate()
  n3 = Power(
   Base=OneMinus(None=MaterialFunctionCall(Float2=@n2) {materialFunction=&BreakOutFloat2Components}.G)
   Exp=DynamicParameter() {defaultValue=<rgba 2 1 1 1>, paramNames [GlowPower,Param2,Param3,Param4]}.GlowPower
  )
  n4 = LinearInterpolate(A=ScalarParameter() {defaultValue=0.2, parameterName=VerticalMaskMultiplier}, B=Constant() {r=1}, Alpha=OneMinus(None=VertexColor().R))
  n5 = ScalarParameter() {parameterName=UseVerticalMask?}
  Multiply(
   A=Multiply(A=Multiply(A=@n3, B=@n1.A), B=LinearInterpolate(B=@n4, Alpha=@n5) {constA=1})
   B=MaterialFunctionCall("FadeSensitivity (optional)"=ScalarParameter() {defaultValue=15, parameterName=FadeAngle}) {materialFunction=&FadeBasedOnViewAngle}.Result
  )
 ]
}
