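# Translucent unlit sprite material for Niagara particles in the DryEraseToro skydiving contrail cosmetic.
# Particle color RGB runs through MF_FX_EmissiveAdjustments for the glow, while particle alpha is multiplied by a selectable channel of a texture parameter, defaulting to a star's red channel.
# A quality switch multiplies that mask by a parameterized camera depth fade on High and Epic only, using a constant one otherwise.
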Material /BRCosmetics/Athena/Effects/Particles/Skydiving/NS_Contrail_DryEraseToro/Materials/M_FX_Sprites
graph {
 properties {
  bAutomaticallySetUsageInEditor=false
  bUsedWithNiagaraSprites=true
  blendMode=BLEND_Translucent
  shadingModel=MSM_Unlit
  usedShadingModels=Unlit
 }
 nodes [
  n1 = ParticleColor()
  MaterialFunctionCall(Emissive=@n1.RGB) {materialFunction=&MF_FX_EmissiveAdjustments}
  n2 = TextureSampleParameter2D() {parameterName=Param, texture=/BRCosmetics/Athena/Effects/Particles/Skydiving/NS_Contrail_DryEraseToro/Textures/T_FX_Star}
  n3 = Constant() {r=1}
  n4 = MaterialFunctionCall(
   Fade Length=ScalarParameter() {defaultValue=200, parameterName=FadeLength}
   Fade Offset=ScalarParameter() {defaultValue=90, parameterName=FadeOffset}
  ) {materialFunction=&CameraDepthFade}
  Multiply(
   A=Multiply(A=@n1.A, B=StaticComponentMaskParameter(None=@n2.RGB) {defaultR=true, parameterName=Channel})
   B=QualitySwitch(Default=@n3, Low=@n3, High=@n4.Result, Medium=@n3, Epic=@n4.Result)
  )
 ]
}
