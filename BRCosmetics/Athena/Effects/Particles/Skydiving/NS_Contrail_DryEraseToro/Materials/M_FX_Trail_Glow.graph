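# Unlit translucent Niagara ribbon material for the glow trail of the DryEraseToro skydiving contrail.
# Particle color drives emissive, while opacity blends two cross-ribbon profiles, scaled by particle alpha and view-angle fade.
# Named reroutes define lengthwise end fades and a dynamic-parameter Reveal wipe for masking.
# Camera depth fade applies only at High and Epic quality, and an optional world-position camera offset is parameterized.
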Material /BRCosmetics/Athena/Effects/Particles/Skydiving/NS_Contrail_DryEraseToro/Materials/M_FX_Trail_Glow
graph {
 properties {
  bAutomaticallySetUsageInEditor=false
  bUsedWithNiagaraRibbons=true
  blendMode=BLEND_Translucent
  shadingModel=MSM_Unlit
  usedShadingModels=Unlit
 }
 nodes [
  LinearInterpolate(
   B=MaterialFunctionCall(Offset Amount=ScalarParameter() {defaultValue=50, parameterName=CameraOffset}) {materialFunction=/Engine/Functions/Engine_MaterialFunctions02/WorldPositionOffset/CameraOffset}.Camera Offset
   Alpha=ScalarParameter() {parameterName=UseCameraOffset}
  )
  n1 = ParticleColor()
  MaterialFunctionCall(Emissive=@n1.RGB) {materialFunction=&MF_FX_EmissiveAdjustments}
  n2 = MaterialFunctionCall(Float2=TextureCoordinate()) {materialFunction=&BreakOutFloat2Components}
  n3 = OneMinus(None=@n2.G)
  n4 = Saturate(None=Power(Base=Subtract(A=@n3) {constB=0.05}) {constExponent=0.4})
  n5 = Multiply(A=Power(Base=Multiply(A=@n4, B=@n2.G))) {constB=1.5}
  n6 = Power(Base=SphereMask(A=@n2.G, B=Constant() {r=0.5}, Radius=Constant() {r=0.5}) {hardnessPercent=0}, Exp=Constant() {r=1.5})
  n7 = ScalarParameter() {defaultValue=1, parameterName=UseSphereMask}
  n8 = Multiply(
   A=Multiply(A=LinearInterpolate(A=@n5, B=@n6, Alpha=@n7), B=NamedRerouteUsage())
   B=MaterialFunctionCall("FadeSensitivity (optional)"=ScalarParameter() {defaultValue=1, parameterName=AngleFade}) {materialFunction=&FadeBasedOnViewAngle}.Result
  )
  n9 = Constant() {r=1}
  n10 = MaterialFunctionCall(
   Fade Length=ScalarParameter() {defaultValue=200, parameterName=FadeLength}
   Fade Offset=ScalarParameter() {defaultValue=50, parameterName=FadeOffset}
  ) {materialFunction=&CameraDepthFade}
  Multiply(
   A=Multiply(A=Multiply(A=@n8, B=@n1.A) {constA=1}, B=NamedRerouteUsage())
   B=QualitySwitch(Default=@n9, Low=@n9, High=@n10.Result, Medium=@n9, Epic=@n10.Result)
  )
  n11 = MaterialFunctionCall(Float2=TextureCoordinate()) {materialFunction=&BreakOutFloat2Components}
  n12 = Multiply(A=OneMinus(None=@n11.R), B=Constant() {r=10})
  n13 = Multiply(A=@n11.R, B=Constant() {r=5})
  NamedRerouteDeclaration(None=Multiply(A=Saturate(None=@n12), B=Saturate(None=@n13))) {name=FadeInOut, nodeColor=<rgba 1 0 0.1641 1>}
  n14 = OneMinus(None=MaterialFunctionCall(Float2=TextureCoordinate()) {materialFunction=&BreakOutFloat2Components}.R)
  n15 = Constant() {r=10}
  NamedRerouteDeclaration(None=Saturate(None=Multiply(A=@n14, B=@n15))) {name=TrailingEdgeFade, nodeColor=<rgba 0 0.0859 1 1>}
  n16 = DynamicParameter() {defaultValue=<rgba -1 1 1 1>, paramNames [Reveal,Param2,Param3,Param4]}
  n17 = Add(A=@n16.Reveal, B=ScalarParameter() {defaultValue=1, parameterName=Softness})
  n18 = OneMinus(None=MaterialFunctionCall(Float2=TextureCoordinate()) {materialFunction=&BreakOutFloat2Components}.R)
  NamedRerouteDeclaration(None=Saturate(None=SmoothStep(Min=@n16.Reveal, Max=@n17, Value=@n18))) {name=Reveal, nodeColor=<rgba 0.2188 0 1 1>}
 ]
}
