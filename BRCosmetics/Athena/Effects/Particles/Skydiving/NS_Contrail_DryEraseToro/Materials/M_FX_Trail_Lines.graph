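# Unlit translucent Niagara ribbon material that draws procedural line trails for the DryEraseToro skydiving contrail.
# Three UV sphere masks form a center line and two mirrored lines whose offset shifts along the ribbon, sharpened by smoothstep without textures.
# Named reroutes share flippable UVs, shape, edge fades and a dynamic-parameter Reveal wipe.
# Particle color feeds emissive adjustments, and its alpha scales opacity, with optional depth fade.
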Material /BRCosmetics/Athena/Effects/Particles/Skydiving/NS_Contrail_DryEraseToro/Materials/M_FX_Trail_Lines
graph {
 properties {
  bAutomaticallySetUsageInEditor=false
  bUsedWithNiagaraRibbons=true
  blendMode=BLEND_Translucent
  shadingModel=MSM_Unlit
  usedShadingModels=Unlit
 }
 nodes [
  n1 = ParticleColor()
  MaterialFunctionCall(Emissive=@n1.RGB) {materialFunction=&MF_FX_EmissiveAdjustments}
  n2 = Constant() {r=0.5}
  n3 = Add(A=@n2, B=Constant() {r=0.5})
  n4 = SphereMask(
   A=MaterialFunctionCall(Float2=NamedRerouteUsage()) {materialFunction=&BreakOutFloat2Components}.G
   B=Constant() {r=0.5}
   Radius=Constant() {r=0.1}
  ) {hardnessPercent=0}
  n5 = MaterialFunctionCall(Float2=NamedRerouteUsage()) {materialFunction=&BreakOutFloat2Components}
  n6 = LinearInterpolate(A=Constant() {r=0.1}, B=Constant() {r=0.38}, Alpha=@n5.R)
  n7 = Multiply(A=SphereMask(A=@n5.G, B=@n6, Radius=Constant() {r=0.1}) {hardnessPercent=0}, B=NamedRerouteUsage())
  n8 = SphereMask(
   A=MaterialFunctionCall(Float2=NamedRerouteUsage()) {materialFunction=&BreakOutFloat2Components}.G
   B=OneMinus(None=@n6)
   Radius=Constant() {r=0.1}
  ) {hardnessPercent=0}
  n9 = NamedRerouteUsage()
  n10 = Saturate(None=Add(A=Add(A=@n4, B=@n7), B=Multiply(A=@n8, B=@n9)))
  NamedRerouteDeclaration(None=Multiply(A=SmoothStep(Min=@n2, Max=@n3, Value=@n10), B=NamedRerouteUsage())) {name=Shape, nodeColor=<rgba 0 0.8594 1 1>}
  n11 = MaterialFunctionCall(Float2=NamedRerouteUsage()) {materialFunction=&BreakOutFloat2Components}
  n12 = Multiply(A=OneMinus(None=@n11.R), B=ScalarParameter() {defaultValue=5, parameterName=Param_4})
  NamedRerouteDeclaration(None=Multiply(A=Saturate(None=@n12), B=Saturate(None=@n11.R))) {name=FadeInOut, nodeColor=<rgba 1 0 0.1641 1>}
  n13 = OneMinus(None=MaterialFunctionCall(Float2=NamedRerouteUsage()) {materialFunction=&BreakOutFloat2Components}.R)
  n14 = Constant() {r=10}
  NamedRerouteDeclaration(None=Saturate(None=Multiply(A=@n13, B=@n14))) {name=TrailingEdgeFade, nodeColor=<rgba 0 0.0859 1 1>}
  n15 = MaterialFunctionCall(Float2=TextureCoordinate()) {materialFunction=&BreakOutFloat2Components}
  n16 = OneMinus(None=@n15.R)
  n17 = ScalarParameter() {parameterName=X Flip}
  NamedRerouteDeclaration(None=MaterialFunctionCall(X=LinearInterpolate(A=@n16, B=@n15.R, Alpha=@n17), Y=@n15.G) {materialFunction=&MakeFloat2}.Result) {name=UV, nodeColor=<rgba 0 1 0.2266 1>}
  n18 = DynamicParameter() {defaultValue=<rgba -1 1 1 1>, paramNames [Reveal,Param2,Param3,Param4]}
  n19 = Add(A=@n18.Reveal, B=ScalarParameter() {defaultValue=1, parameterName=Softness})
  n20 = OneMinus(None=MaterialFunctionCall(Float2=TextureCoordinate()) {materialFunction=&BreakOutFloat2Components}.R)
  NamedRerouteDeclaration(None=Saturate(None=SmoothStep(Min=@n18.Reveal, Max=@n19, Value=@n20))) {name=Reveal, nodeColor=<rgba 0.2188 0 1 1>}
  n21 = Multiply(A=Multiply(A=NamedRerouteUsage(), B=@n1.A), B=NamedRerouteUsage())
  StaticSwitchParameter(True=DepthFade(Opacity=@n21, FadeDistance=ScalarParameter() {parameterName=FadeDistance}), False=@n21) {parameterName=UseDepthFade}
  TextureCoordinate()
 ]
}
