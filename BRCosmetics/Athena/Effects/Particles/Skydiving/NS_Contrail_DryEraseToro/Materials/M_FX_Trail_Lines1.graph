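# Translucent unlit Niagara ribbon material that draws soft procedural trail lines for the DryEraseToro skydiving contrail.
# Sphere masks across UV V form a center line and mirrored side pairs, which are summed and smoothstepped into a shape.
# Named reroutes define U-axis fades and a dynamic-parameter Reveal wipe.
# Particle color feeds emissive adjustment, while particle alpha scales opacity, with camera depth fade only on High and Epic quality.
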Material /BRCosmetics/Athena/Effects/Particles/Skydiving/NS_Contrail_DryEraseToro/Materials/M_FX_Trail_Lines1
graph {
 properties {
  bAutomaticallySetUsageInEditor=false
  bUsedWithNiagaraRibbons=true
  blendMode=BLEND_Translucent
  shadingModel=MSM_Unlit
  usedShadingModels=Unlit
 }
 nodes [
  Constant() {r=1}
  n1 = ParticleColor()
  MaterialFunctionCall(Emissive=@n1.RGB) {materialFunction=&MF_FX_EmissiveAdjustments}
  n2 = MaterialFunctionCall(Float2=TextureCoordinate()) {materialFunction=&BreakOutFloat2Components}
  MaterialFunctionCall(X=@n2.R, Y=@n2.R, Z=Constant()) {materialFunction=&MakeFloat3}
  n3 = NamedRerouteUsage()
  n4 = Constant() {r=1}
  n5 = MaterialFunctionCall(
   Fade Length=ScalarParameter() {defaultValue=200, parameterName=FadeLength}
   Fade Offset=ScalarParameter() {defaultValue=60, parameterName=FadeOffset}
  ) {materialFunction=&CameraDepthFade}
  n6 = Multiply(A=@n1.A, B=QualitySwitch(Default=@n4, Low=@n4, High=@n5.Result, Medium=@n4, Epic=@n5.Result))
  Multiply(A=Saturate(None=Multiply(A=@n3, B=@n6) {constA=1}))
  n7 = ScalarParameter() {defaultValue=0.5, parameterName=Param_5}
  n8 = MaterialFunctionCall(Float2=TextureCoordinate()) {materialFunction=&BreakOutFloat2Components}
  n9 = ScalarParameter() {defaultValue=0.5}
  n10 = ScalarParameter() {defaultValue=0.1, parameterName=Param_1}
  n11 = ScalarParameter() {defaultValue=0.1, parameterName=Param_2}
  n12 = SphereMask(
   A=MaterialFunctionCall(Float2=TextureCoordinate()) {materialFunction=&BreakOutFloat2Components}.G
   B=@n11
   Radius=ScalarParameter() {defaultValue=0.1, parameterName=Param_1}
  ) {hardnessPercent=0}
  n13 = NamedRerouteUsage()
  n14 = MaterialFunctionCall(Float2=TextureCoordinate()) {materialFunction=&BreakOutFloat2Components}
  n15 = OneMinus(None=@n11)
  n16 = ScalarParameter() {defaultValue=0.1, parameterName=Param_1}
  n17 = Add(
   A=Add(A=SphereMask(A=@n8.G, B=@n9, Radius=@n10) {hardnessPercent=0}, B=Multiply(A=@n12, B=@n13))
   B=Multiply(A=SphereMask(A=@n14.G, B=@n15, Radius=@n16) {hardnessPercent=0}, B=NamedRerouteUsage())
  )
  n18 = TextureCoordinate()
  n19 = ScalarParameter() {defaultValue=0.29, parameterName=Param_8}
  n20 = TextureCoordinate()
  n21 = Add(
   A=SphereMask(
    A=MaterialFunctionCall(Float2=@n18) {materialFunction=&BreakOutFloat2Components}.G
    B=@n19
    Radius=ScalarParameter() {defaultValue=0.1, parameterName=Param_1}
   ) {hardnessPercent=0}
   B=SphereMask(
    A=MaterialFunctionCall(Float2=@n20) {materialFunction=&BreakOutFloat2Components}.G
    B=OneMinus(None=@n19)
    Radius=ScalarParameter() {defaultValue=0.1, parameterName=Param_1}
   ) {hardnessPercent=0}
  )
  n22 = SmoothStep(Min=@n7, Max=Add(A=@n7, B=ScalarParameter() {defaultValue=0.5, parameterName=Param_6}), Value=Saturate(None=Add(A=@n17, B=@n21)))
  n23 = NamedRerouteUsage()
  NamedRerouteDeclaration(None=Multiply(A=Multiply(A=@n22, B=@n23), B=NamedRerouteUsage())) {name=Shape, nodeColor=<rgba 0 0.8594 1 1>}
  n24 = MaterialFunctionCall(Float2=TextureCoordinate()) {materialFunction=&BreakOutFloat2Components}
  n25 = Multiply(A=OneMinus(None=@n24.R), B=ScalarParameter() {defaultValue=5, parameterName=Param_4})
  NamedRerouteDeclaration(None=Multiply(A=Saturate(None=@n25), B=Saturate(None=@n24.R))) {name=FadeInOut, nodeColor=<rgba 1 0 0.1641 1>}
  n26 = OneMinus(None=MaterialFunctionCall(Float2=TextureCoordinate()) {materialFunction=&BreakOutFloat2Components}.R)
  n27 = ScalarParameter() {defaultValue=10, parameterName=Param_7}
  NamedRerouteDeclaration(None=Saturate(None=Multiply(A=@n26, B=@n27))) {name=TrailingEdgeFade, nodeColor=<rgba 0 0.0859 1 1>}
  n28 = DynamicParameter() {defaultValue=<rgba -1 1 1 1>, paramNames [Reveal,Param2,Param3,Param4]}
  n29 = Add(A=@n28.Reveal, B=ScalarParameter() {defaultValue=1, parameterName=Softness})
  n30 = OneMinus(None=MaterialFunctionCall(Float2=TextureCoordinate()) {materialFunction=&BreakOutFloat2Components}.R)
  NamedRerouteDeclaration(None=Saturate(None=SmoothStep(Min=@n28.Reveal, Max=@n29, Value=@n30))) {name=Reveal, nodeColor=<rgba 0.2188 0 1 1>}
  NamedRerouteUsage()
  ScalarParameter() {defaultValue=1, parameterName=Param_3}
 ]
}
