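# Skydiving contrail that trails golden ribbons, musical notes, a bending road of keys and pulsing glows from the player's skeletal mesh.
# BottomTrails and Keys share a mesh-relative wavy path, bent by a scratch road-bending module and reshaped by glide lean.
# HandRibbons build six lagged, swaying ribbons from two sockets, which HandNotes sprites follow through particle reads.
# System-level dive, fade, reveal and audio values drive motion and brightness, with separate front-end tuning.
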NiagaraSystem /BRCosmetics/Athena/Effects/Particles/Skydiving/NS_Contrail_DryEraseToro/NS_Contrail_DryEraseToro
dependencies {
 usedDataInterfaces [
  &NiagaraDataInterfaceCurve
  &NiagaraDataInterfaceArrayFloat
  &NiagaraDataInterfaceParticleRead
  &NiagaraDataInterfaceArrayFloat3
  &NiagaraDataInterfaceColorCurve
  &NiagaraDataInterfaceSkeletalMesh
  &NiagaraDataInterfaceSocketReader
 ]
 usedDynamicInputs [
  /Niagara/DynamicInputs/Integer/V2/Round_Float
  &FN_FrontEndVector3Modifier
  &SetBoolByFloatComparison
  &self:Abs_Z_Axis
  &Abs_Float
  &MakeFloatFromVector
  &Divide_Float
  &NormalizeVector
 ]
 usedModules [
  &self:SystemSpawnScript.SystemScriptSource.SystemScriptGraph.NiagaraNodeAssignment_4.SetVariables_#1
  &self:SystemSpawnScript.SystemScriptSource.SystemScriptGraph.NiagaraNodeAssignment_6.SetVariables_#2
  &self:SystemSpawnScript.SystemScriptSource.SystemScriptGraph.NiagaraNodeAssignment_5.SetVariables_#3
  /Game/Effects/Niagara/Modules/Skydiving/FN_DivingAttributes/v5/FN_DivingAttributes
  &self:SystemSpawnScript.SystemScriptSource.SystemScriptGraph.NiagaraNodeAssignment_1.SetVariables_#4
  &SystemState
  &FN_Integrator
  &self:SystemSpawnScript.SystemScriptSource.SystemScriptGraph.NiagaraNodeAssignment_2.SetVariables_#5
  &Timeline
  &self:SystemSpawnScript.SystemScriptSource.SystemScriptGraph.NiagaraNodeAssignment_7.SetVariables_#6
  &self:SystemSpawnScript.SystemScriptSource.SystemScriptGraph.NiagaraNodeAssignment_8.SetVariables_#7
  &self:SystemSpawnScript.SystemScriptSource.SystemScriptGraph.NiagaraNodeAssignment_10.SetVariables_#8
 ]
}
emitters {
 BottomTrails {
  emitterSpawnScript [/SetVariables(Emitter.ParticlesPerTrail=30) {isSetParameters=true}]
  emitterUpdateScript [
   &EmitterState(Life Cycle Mode=::NewEnumerator1, Loop Behavior=::NewEnumerator1, Loop Duration Mode=::NewEnumerator1, UseLoopDelay=false)
   &SpawnBurst_Instantaneous(
    Spawn Count=&Multiply_Int(A=$Emitter.ParticlesPerTrail, B=2)
    Use Spawn Probability=false
    UseLoopCountLimit=true
   )
  ]
  particleSpawnScript [
   &PartitionParticles(
    Normalized Index In Partition Includes One=true
    Particles Per Partition:NiagaraInt32=$Emitter.ParticlesPerTrail
    Partition Mode:ENiagara_PartitionMode_v2=::NewEnumerator1
    Write Partition Index to Ribbon Link Order:ENiagara_AttributeSamplingApplyOutput=::NewEnumerator0
    Write Partition to Ribbon ID=::NewEnumerator0
   )
   &InitializeParticle(
    Color=<rgba 1 0.7421 0 0.6>
    Lifetime=5
    Mesh Scale=<0.6 0.4 1>
    Mesh Scale Mode=::NewEnumerator4
    UsePositionOffset=false
    Write Material Random=true
   )
   /SetVariables() {isSetParameters=true, module=SetVariables_#9}
   /SetVariables(Particles.GizmoScale:Vector3f=<5 5 5>) {isSetParameters=true, module=SetVariables_#10}
   /SetVariables(Particles.NormalizedParticleIndex:NiagaraFloat=&Power(a=&OneMinusFloat(Float=$Particles.PartitionParticles.Normalized Index In Partition), b=3)) {isSetParameters=true, module=SetVariables_#11}
  ]
  particleUpdateScript [
   &ParticleState(Kill Particles When Lifetime Has Elapsed=false, Let Infinitely Lived Particles Die When Emitter Deactivates=true)
   /SetVariables(
    Particles.RibbonFacing=&TransformVector(Vector=<0 0 1>, Source Space=::World, Destination Space=::Local)
    Particles.RibbonWidth=&Lerp_Float(
     A=60
     B=15
     Alpha=&FloatFromCurve(
      FloatCurve=DataInterface() {propertyValues {Curve {keys [{arriveTangent=-2.079, interpMode=RCIM_Cubic, leaveTangent=-2.079, tangentMode=RCTM_User, time=0, value=1}, <at 0.7 0 RCIM_Cubic>]}}}
      CurveIndex=$Particles.PartitionParticles.Normalized Index In Partition
     )
    )
   ) {isSetParameters=true}
   /SetVariables(Particles.PreviousPos:NiagaraPosition=$Particles.Position) {isSetParameters=true, module=SetVariables_#12}
   &SkeletalMeshLocation(
    Bone Sampling Mode=::NewEnumerator4
    MaskSpawnGroup=false
    Orientation Sampling=::NewEnumerator1
    Sampled Position Offset=&Add_Vector(
     A=&MakeVector(
      X=&Add_Float(
       A=&FN_FrontEndFloatModifier(Gameplay=-70, Front End=-50)
       B=&Multiply_Float(A=$Particles.PartitionParticles.Normalized Index In Partition, B=-300)
      )
      Y=&Add_Float(
       A=&Lerp_Float(A=65, B=50, Alpha=$User.ParaGlideLeanAlpha)
       B=&Lerp_Float(
        A=&Add_Float(
         A=&Multiply_Float(
          A=&FloatFromCurve(
           FloatCurve=DataInterface() {propertyValues {Curve {
              defaultValue=1
              keys [
               {arriveTangent=-2.4782, interpMode=RCIM_Cubic, leaveTangent=-2.4782, leaveTangentWeight=0.1667, tangentMode=RCTM_User, time=0, value=1.3}
               {arriveTangentWeight=0.1667, interpMode=RCIM_Cubic, leaveTangentWeight=0.1667, tangentMode=RCTM_User, time=0.5, value=0}
               {arriveTangent=1.6086, tangentMode=RCTM_User, time=1, value=1}
              ]
             }}}
           CurveIndex=$Particles.PartitionParticles.Normalized Index In Partition
          )
          B=-25
         )
         B=&Multiply_Float(
          A=&Sine(
           Normalized Angle=&Multiply_Float(
            A=&Add_Float(A=$System.Age, B=&Multiply_Float(A=$Particles.PartitionParticles.Normalized Index In Partition, B=2))
            B=-1
           )
           Period=1
           Scale=1
           Bias=0
          )
          B=5
         )
        )
        B=&Multiply_Float(
         A=&FloatFromCurve(
          FloatCurve=DataInterface() {propertyValues {Curve {
             defaultValue=1
             keys [
              {arriveTangent=-1.2393, leaveTangentWeight=0.1667, time=0, value=0}
              {
               arriveTangent=-0.9143
               arriveTangentWeight=0.1667
               interpMode=RCIM_Cubic
               leaveTangentWeight=0.1667
               tangentMode=RCTM_Break
               time=0.5
               value=0
              }
              {arriveTangent=3.1583, tangentMode=RCTM_User, time=1, value=1}
             ]
            }}}
          CurveIndex=$Particles.PartitionParticles.Normalized Index In Partition
         )
         B=-50
        )
        Alpha=$User.ParaGlideLeanAlpha
       )
      )
      Z=&Add_Float(
       A=&ClampFloat(
        Float=&Multiply_Float(
         A=$Particles.NormalizedParticleIndex
         B=&RemapRange(Input Value To Remap=$Particles.Roadbending, Input Min=-250, Input Max=250, Output Min=100, Output Max=-100)
        )
        Min=-100
        Max=100
       )
       B=0
      )
     )
     B=&Add_Vector(
      A=&AddVectorToPosition(Position=Linked() {linkedVariable {name=System.MasterRoadOffset, type=Vector3f}}, Vector=<0 0 0>)
      B=&MakeVector(
       X=-20
       Y=0
       Z=&Lerp_Float(
        A=&MergeHelper_FloatSwitch(0=-80, 1=80, 2=?, 3=?, Index=$Particles.PartitionParticles.Partition, Num=::NewEnumerator0)
        B=&Lerp_Float(
         A=&MergeHelper_FloatSwitch(0=-60, 1=60, 2=?, 3=?, Index=$Particles.PartitionParticles.Partition, Num=::NewEnumerator0)
         B=&MergeHelper_FloatSwitch(0=-35, 1=35, 2=?, 3=?, Index=$Particles.PartitionParticles.Partition, Num=::NewEnumerator0)
         Alpha=&RemapRange(
          Input Value To Remap=$Particles.Roadbending
          Input Min=0
          Input Max=&MergeHelper_FloatSwitch(0=-100, 1=100, 2=?, 3=?, Index=$Particles.PartitionParticles.Partition, Num=::NewEnumerator0)
          Output Min=1
          Output Max=0
         )
        )
        Alpha=&FloatFromCurve(
         FloatCurve=DataInterface() {propertyValues {Curve {keys [{arriveTangent=-1.5648, time=0, value=1}, {leaveTangent=-0.6189, time=1, value=0}]}}}
         CurveIndex=$Particles.PartitionParticles.Normalized Index In Partition
        )
       )
      )
     )
    )
    Skeletal Mesh=$System.Skeletal Mesh
    Use Sampled Mesh Attribute as Ribbon ID=false
   )
   &self:SetRoadBending(CurrentPos:NiagaraPosition=$Particles.Position, MaxBending=250, RestSpeed=1, VelDivider=4) {module=SetRoadBending}
   &ScaleColor(
    Scale RGB=&VectorFromFloat(Value=&RemapRange(
      Input Value To Remap=&Sine(
       Normalized Angle=&Add_Float(
        A=&Multiply_Float(A=$Particles.PartitionParticles.Normalized Index In Partition, B=3)
        B=&Multiply_Float(
         A=&Add_Float(A=$Engine.System.Age, B=&Multiply_Float(A=$System.FN_Integrator002.Value, B=10))
         B=-1
        )
       )
       Period=1
       Scale=1
       Bias=0
      )
      Input Min=-1
      Input Max=1
      Output Min=0.1
      Output Max=1
     ))
    ScaleA=false
    ScaleRGB=true
   )
   &ScaleColor(
    Scale Alpha=&Multiply_Float(
     A=1
     B=&FloatFromCurve(
      FloatCurve=DataInterface() {propertyValues {Curve {keys [<at 0.3 1 RCIM_Cubic>,<at 1 0>]}}}
      CurveIndex=$Particles.Initial.PartitionParticles.Normalized Index In Partition
     )
    )
    Scale RGB=&VectorFromFloat(Value=&Multiply_Float(
      A=&Lerp_Float(A=1, B=6, Alpha=$Particles.NormalizedParticleIndex)
      B=&FloatFromCurve(
       FloatCurve=DataInterface() {propertyValues {Curve {keys [
           <at 0 1 RCIM_Cubic>
           {arriveTangent=5.8588, interpMode=RCIM_Cubic, leaveTangent=5.8588, time=0.2, value=6}
           {arriveTangent=-12.5901, interpMode=RCIM_Cubic, leaveTangent=-12.5901, tangentMode=RCTM_User, time=0.3453, value=3.0232}
           <at 0.8 1 RCIM_Cubic>
          ]}}}
       CurveIndex=$System.Timeline003.NormalizedTime
      )
     ))
    ScaleA=true
    ScaleRGB=true
   ) {module=ScaleColor001}
   &FN_Sprite_Attributes(
    Color=&MultiplyLinearColor(
     Linear Color=<rgba 0.087 0.0181 0 1>
     Scale Factor=&MakeVector4FromVectorAndFloat(
      "Vector (RGB)"=<1 1 1>
      "Float (Alpha)"=&Multiply_Float(
       A=1
       B=&FloatFromCurve(
        FloatCurve=DataInterface() {propertyValues {Curve {keys [{arriveTangent=-4.0505, interpMode=RCIM_Cubic, leaveTangent=-4.0505, tangentMode=RCTM_User, time=0.3, value=1}, <at 1 0>]}}}
        CurveIndex=$Particles.PartitionParticles.Normalized Index In Partition
        Scale Curve=0.6
       )
      )
     )
    )
    Create Color Attribute=true
    Write to Module Namespace=true
   )
   &FN_Sprite_Attributes(
    Color=&MultiplyLinearColor(
     Linear Color=<rgba 1 0.6042 0.05 1>
     Scale Factor=&MakeVector4FromVectorAndFloat(
      "Vector (RGB)"=<1 1 1>
      "Float (Alpha)"=&Multiply_Float(
       A=1
       B=&FloatFromCurve(
        FloatCurve=DataInterface() {propertyValues {Curve {keys [{arriveTangent=-4.0505, interpMode=RCIM_Cubic, leaveTangent=-4.0505, tangentMode=RCTM_User, time=0.3, value=1}, <at 1 0>]}}}
        CurveIndex=$Particles.PartitionParticles.Normalized Index In Partition
        Scale Curve=0.4
       )
      )
     )
    )
    Create Color Attribute=true
    Write to Module Namespace=true
   ) {module=FN_Sprite_Attributes001}
   &DynamicMaterialParameters(
    Index 0 Param 1=$System.RevealCurve
    Module.Index0Float0Write=true
    Module.Index0Float1Write=false
    Module.Index0Float2Write=false
    Module.Index0Float3Write=false
   )
  ]
  rendererClasses [&NiagaraRibbonRendererProperties, &NiagaraMeshRendererProperties]
  renderers [
   {rendererClass=&NiagaraRibbonRendererProperties}
   {rendererClass=&NiagaraRibbonRendererProperties, rendererIndex=1}
   {rendererClass=&NiagaraRibbonRendererProperties, rendererIndex=2}
   {rendererClass=&NiagaraMeshRendererProperties, rendererIndex=3}
  ]
 }
 Decal_Glow001 {
  emitterUpdateScript [
   &EmitterState(Life Cycle Mode=::NewEnumerator1, Loop Behavior=::NewEnumerator1, Loop Duration Mode=::NewEnumerator1, UseLoopDelay=false)
   &SpawnBurst_Instantaneous(Spawn Count=2, Use Spawn Probability=false, UseLoopCountLimit=false)
  ]
  particleSpawnScript [
   &InitializeParticle(
    Color=<rgba 2 1.25 0.2 0.2>
    Lifetime=&RandomRangeFloat(Minimum=1, Maximum=2)
    Sprite Rotation Mode=::NewEnumerator2
    Sprite Size Mode=::NewEnumerator3
    Sprite UV Mode=::NewEnumerator3
    Uniform Sprite Size=&RandomRangeFloat(Minimum=75, Maximum=100)
    UsePositionOffset=false
    Write Material Random=true
   )
   &PartitionParticles(Number Of Partitions=4, Write Partition to Ribbon ID=::NewEnumerator0)
  ]
  particleUpdateScript [
   &ParticleState(
    Kill Particles When Lifetime Has Elapsed=false
    Let Infinitely Lived Particles Die When Emitter Deactivates=true
    Loop Particles Lifetime=true
   )
   &SkeletalMeshLocation(
    Bone / Socket Index=&Add_Integer(A=1, B=$Particles.PartitionParticles.Partition)
    Bone Sampling Mode=::NewEnumerator4
    MaskSpawnGroup=false
    Orientation Sampling=::NewEnumerator1
    Skeletal Mesh=$System.Skeletal Mesh
   )
   &SocketLocation(
    SamplePosOffset=false
    Socket Index=$Particles.PartitionParticles.Partition
    Socket Reader=$System.Socket Reader
    Socket Sampling Mode=::NewEnumerator3
    VelocitySpeedLimit=false
   ) {enabled=false}
   &CameraOffset(Camera Offset Amount=10)
   &ScaleColor(
    Scale Alpha=&Multiply_Float(A=$System.FadeInCurve, B=&FN_FrontEndFloatModifier(Gameplay=1, Front End=0.25))
    Scale RGB=&Multiply_Vector(
     A=&MakeVectorFromLinearColorRGB(Color=<rgba 1 1 1 1>)
     B=&VectorFromFloat(Value=1)
    )
    ScaleA=true
    ScaleRGB=true
   )
   &ScaleSpriteSize(Scale Factor=&Vector2DFromFloat(Value=&Multiply_Float(
      A=&FloatFromCurve(
       FloatCurve=DataInterface() {propertyValues {Curve {keys [
           {interpMode=RCIM_Cubic, tangentMode=RCTM_User, time=0, value=1}
           {interpMode=RCIM_Cubic, tangentMode=RCTM_User, time=0.5, value=1.1}
           {tangentMode=RCTM_User, time=1, value=1}
          ]}}}
       Scale Curve=&SelectFloatFromArray(
        Direct Array Index=$Particles.PartitionParticles.Partition
        Float Selection Array=DataInterface() {propertyValues {FloatData [0.8,0.8,0.4,0.4]}}
       )
      )
      B=&Lerp_Float(A=1, B=1.5, Alpha=$System.SmoothedAudioScale)
     )))
   &FN_Sprite_Attributes(
    Color=&MultiplyLinearColor(
     Linear Color=&MultiplyLinearColor(
      Linear Color=<rgba 1 0.625 0.1 1>
      Scale Factor=&MakeFloat4FromVectorAndFloat(
       "Vector (RGB)"=<1 1 1>
       "Float (Alpha)"=&Multiply_Float(
        A=&Multiply_Float(A=0.5, B=&FN_FrontEndFloatModifier(Gameplay=1, Front End=0.25))
        B=$System.FadeInCurve
       )
      )
     )
     Scale Factor=<1 1 1 1>
    )
    Create Color Attribute=true
    Create SpriteSize Attribute=true
    SpriteSize=&Multiply_Vector2D(A=$Particles.SpriteSize, B=<1.5 1.5>)
    Write to Module Namespace=true
   )
   &DynamicMaterialParameters(
    Index 0 Param 1=1.5
    Index 0 Param 3=5
    Index 0 Param 4=&RandomRangeFloat(Minimum=0, Maximum=1)
    Index 2 Param 1=3
    Index 2 Param 2=1
    Index 2 Param 3=1
    Index 2 Param 4=1
    Module.Param0WriteEnabled=true
    Module.Param0WriteEnabled2=true
    Module.Param1WriteEnabled=true
    Module.Param1WriteEnabled2=true
    Module.Param2WriteEnabled=true
    Module.Param2WriteEnabled2=true
    Module.Param3WriteEnabled=true
    Module.Param3WriteEnabled2=true
    Write Parameter Index 2=true
   )
  ]
  rendererClasses [&NiagaraSpriteRendererProperties]
  renderers [
   {rendererClass=&NiagaraSpriteRendererProperties}
   {rendererClass=&NiagaraSpriteRendererProperties, rendererIndex=1}
  ]
 }
 FN_Debug {
  bEnabled=false
  emitterUpdateScript [
   &EmitterState()
   &SpawnBurst_Instantaneous(Spawn Count=1, Use Spawn Probability=false)
  ]
  particleSpawnScript [/Niagara/Modules/Spawn/Initialization/InitializeParticle(
    Lifetime=5
    Module.Write Color=true
    Module.Write Lifetime=true
    Module.Write Mass=true
    Module.Write Material Random=true
    Module.Write Position=false
    Module.Write Scale=false
    Module.Write SpriteRotation=false
    Module.Write SpriteSize=true
    Position=&SimulationPosition()
    Sprite Size=<100 100>
   )]
  particleUpdateScript [
   &ParticleState()
   &SystemLocation(MaskSpawnGroup=false, Offset=<150 0 0>)
   /SetVariables(Particles.Debug:NiagaraFloat=&FN_FrontEndFloatModifier(Gameplay=0, Front End=0.25)) {isSetParameters=true}
   &DynamicMaterialParameters(
    Index 0 Param 1=$Particles.Debug
    Module.Param0WriteEnabled=true
    Module.Param1WriteEnabled=true
    Module.Param2WriteEnabled=true
    Module.Param3WriteEnabled=true
   )
  ]
  rendererClasses [&NiagaraSpriteRendererProperties]
  renderers [{rendererClass=&NiagaraSpriteRendererProperties}]
 }
 HandNotes {
  emitterUpdateScript [
   &EmitterState(
    Delay First Loop Only=true
    Inactive Response=::NewEnumerator1
    Life Cycle Mode=::NewEnumerator1
    Loop Delay=0.6
    Loop Duration=1
    UseLoopDelay=true
   )
   &SpawnBurst_Instantaneous(Spawn Count=10, Spawn Time=0.1, Use Spawn Probability=false, UseLoopCountLimit=false) {enabled=false}
   &SpawnRate(
    SpawnRate=&Multiply_Float(
     A=24
     B=&Lerp_Float(A=1, B=&FN_FrontEndFloatModifier(Gameplay=4, Front End=2), Alpha=$User.ParaGlideLeanAlpha)
    )
    Use Spawn Probability=false
   )
  ]
  particleSpawnScript [
   /SetVariables(Particles.RandomPerParticle:NiagaraFloat=&UniformAOrBFloat(A=0, Evaluation Type=::NewEnumerator0, Recalculate Random Each Loop=true, Override Seed=false)) {isSetParameters=true, module=SetVariables_#13}
   &FN_SetSpawnGroup(Number of Spawn Groups=6)
   /Niagara/Modules/Spawn/Initialization/V2/InitializeParticle(
    Color=<rgba 2 1.4945 0.04 1>
    Lifetime=&Multiply_Float(
     A=&RandomRangeFloat(Minimum=5, Maximum=4)
     B=&Lerp_Float(A=1, B=0.3, Alpha=$User.ParaGlideLeanAlpha)
    )
    Sprite Rotation Angle=&RandomRangeFloat(Minimum=-45, Maximum=45)
    Sprite Rotation Mode=::NewEnumerator0
    Sprite Size Mode=::NewEnumerator3
    Uniform Sprite Size=&Multiply_Float(
     A=&RandomRangeFloat(Minimum=30, Maximum=50)
     B=&Lerp_Float(A=1, B=0.3, Alpha=$Particles.RandomPerParticle)
    )
    UsePositionOffset=false
    Write Material Random=true
   ) {module=InitializeParticle001}
   &DynamicMaterialParameters(
    Index 0 Param 1=&UniformAOrBFloat(A=0.333, Evaluation Type=::NewEnumerator0, Recalculate Random Each Loop=true, Override Seed=false)
    Module.Index0Param0Write=true
    Module.Index0Param1Write=false
    Module.Index0Param2Write=false
    Module.Index0Param3Write=false
   )
   /SetVariables(
    Particles.SubImageIndexNotes:NiagaraFloat=&RandomRangeFloat(Minimum=0, Maximum=3)
    Particles.VisibilityTag=&MakeIntFromFloat(Float=$Particles.RandomPerParticle)
   ) {isSetParameters=true, module=SetVariables_#14}
  ]
  particleUpdateScript [
   &ParticleState() {module=ParticleState001}
   &self:FollowRibbon(
    Alpha:NiagaraFloat=&FloatFromCurve(FloatCurve=DataInterface() {propertyValues {Curve {keys [<at 0 0>,<at 1 1>]}}})
    ParticlePerRibbon=10
    RibbonNumber:NiagaraInt32=$Particles.SpawnGroup
    RibbonReader=DataInterface() {propertyValues {EmitterBinding {bindingMode=Other, emitterName=HandRibbons}}}
   ) {module=FollowRibbon}
   /SetVariables(Particles.Position=&AddVectorToPosition(
     Position=&Multiply_Vector(
      A=Linked() {linkedVariable {name=Output.FollowRibbon.RibbonPos, type=NiagaraPosition}}
      B=&VectorFromFloat(Value=1)
     )
     Vector=&Multiply_VectorByFloat(
      Vector=&RandomRangeVector(Minimum=<-1 -1 -1>, Maximum=<1 1 1>)
      Float=&Multiply_Float(
       A=&FloatFromCurve(FloatCurve=DataInterface() {propertyValues {Curve {keys [<at 0 0.5 RCIM_Cubic>, {arriveTangent=-0.5, interpMode=RCIM_Cubic, leaveTangent=-0.5, time=0.5, value=1}, <at 1 0 RCIM_Cubic>]}}})
       B=20
      )
     )
    )) {isSetParameters=true, module=SetVariables_#15}
   &ParticleState()
   &SpriteRotationRate(Rotation Rate=&Lerp_Float(
     A=&UniformAOrBFloat(A=-20, B=20, Evaluation Type=::NewEnumerator0, Recalculate Random Each Loop=true, Override Seed=false)
     B=&UniformAOrBFloat(A=-100, B=100, Evaluation Type=::NewEnumerator0, Recalculate Random Each Loop=true, Override Seed=false)
     Alpha=$Particles.RandomPerParticle
    ))
   &CameraOffset(Camera Offset Amount=5)
   &ScaleColor(Scale Alpha=$System.FadeInCurveHands, ScaleA=true, ScaleRGB=false)
   /SetVariables(Particles.SpriteSize=&Multiply_Vector2D_ByFloat(
     Float=&Multiply_Float(
      A=&FloatFromCurve(FloatCurve=DataInterface() {propertyValues {Curve {keys [<at 0 0.5>,<at 0.7 1>]}}})
      B=&UniformAOrBFloat(
       A=&FloatFromCurve(FloatCurve=DataInterface() {propertyValues {Curve {keys [
            <at 0 0 RCIM_Cubic>
            {arriveTangent=9.6335, interpMode=RCIM_Cubic, leaveTangent=2.7827, tangentMode=RCTM_Break, time=0.1, value=1}
            {arriveTangent=1.0943, interpMode=RCIM_Cubic, leaveTangent=1.0943, time=0.3322, value=0.4478}
            {arriveTangent=4.272, interpMode=RCIM_Cubic, leaveTangent=0.772, tangentMode=RCTM_Break, time=0.5569, value=1.5}
            {arriveTangent=-1.4573, interpMode=RCIM_Cubic, leaveTangent=-1.4573, time=0.6943, value=0.7273}
            {arriveTangent=2.7202, interpMode=RCIM_Cubic, leaveTangent=-2.3793, tangentMode=RCTM_Break, time=0.9, value=1}
            <at 0.95 0 RCIM_Cubic>
           ]}}})
       B=&FloatFromCurve(FloatCurve=DataInterface() {propertyValues {Curve {keys [
            <at 0 0 RCIM_Cubic>
            {arriveTangent=9.6335, interpMode=RCIM_Cubic, leaveTangent=2.7827, tangentMode=RCTM_Break, time=0.1, value=0.9794}
            {arriveTangent=0.0257, interpMode=RCIM_Cubic, leaveTangent=0.0257, time=0.5397, value=0.5}
            {arriveTangent=2.7202, interpMode=RCIM_Cubic, leaveTangent=-2.3793, tangentMode=RCTM_Break, time=0.9, value=1}
            <at 0.95 0 RCIM_Cubic>
           ]}}})
       Evaluation Type=::NewEnumerator0
       Recalculate Random Each Loop=true
       Override Seed=false
      )
     )
     Vector2D=$Particles.Initial.SpriteSize
    )) {isSetParameters=true, module=SetVariables_#16}
   &FN_Sprite_Attributes(
    Color=&MultiplyLinearColor(
     Linear Color=<rgba 2 0.7333 0.1 1>
     Scale Factor=&MakeVector4FromVectorAndFloat("Vector (RGB)"=<1 1 1>, "Float (Alpha)"=$System.FadeInCurveHands)
    )
    Create Color Attribute=true
    Write to Module Namespace=true
   ) {module=FN_Sprite_Attributes001}
   &FN_Sprite_Attributes(
    Color=&MultiplyLinearColor(
     Linear Color=<rgba 0.087 0.0181 0 1>
     Scale Factor=&MakeVector4FromVectorAndFloat("Vector (RGB)"=<1 1 1>, "Float (Alpha)"=&Multiply_Float(A=$System.FadeInCurveHands, B=1))
    )
    Create Color Attribute=true
    Write to Module Namespace=true
   )
  ]
  rendererClasses [&NiagaraSpriteRendererProperties]
  renderers [
   {rendererClass=&NiagaraSpriteRendererProperties}
   {rendererClass=&NiagaraSpriteRendererProperties, rendererIndex=1}
   {rendererClass=&NiagaraSpriteRendererProperties, rendererIndex=2}
   {rendererClass=&NiagaraSpriteRendererProperties, rendererIndex=3}
   {rendererClass=&NiagaraSpriteRendererProperties, rendererIndex=4}
   {rendererClass=&NiagaraSpriteRendererProperties, rendererIndex=5}
  ]
 }
 HandRibbons {
  emitterSpawnScript [
   /SetVariables(Emitter.InterpolatedVector:Vector3f=<0 0 1>, Emitter.TargetVector:Vector3f=<0 1 1>) {isSetParameters=true}
   /SetVariables(Emitter.RibbonLength=275) {isSetParameters=true}
   /SetVariables(Emitter.ParticlesPerTrail=10) {isSetParameters=true}
  ]
  emitterUpdateScript [
   &EmitterState(
    Inactive Response=::NewEnumerator1
    Life Cycle Mode=::NewEnumerator1
    Loop Behavior=::NewEnumerator1
    Loop Duration Mode=::NewEnumerator1
    UseLoopDelay=false
   )
   &SpawnBurst_Instantaneous(
    Spawn Count=&Multiply_Int(A=6, B=$Emitter.ParticlesPerTrail)
    Use Spawn Probability=false
    UseLoopCountLimit=false
   )
   &self:ScratchModule_05(Target Vector:Vector3f=$System.LookAtVector) {module=ScratchModule_05}
  ]
  particleSpawnScript [
   &PartitionParticles(
    Normalized Index In Partition Includes One=true
    Number Of Partitions=6
    Particles Per Partition:NiagaraInt32=$Emitter.ParticlesPerTrail
    Partition Mode:ENiagara_PartitionMode_v2=::NewEnumerator1
    Write Partition Index to Ribbon Link Order:ENiagara_AttributeSamplingApplyOutput=::NewEnumerator0
    Write Partition to Ribbon ID=::NewEnumerator0
   )
   /SetVariables() {isSetParameters=true}
  ]
  particleUpdateScript [
   &SocketLocation(
    Orientation Sampling=::NewEnumerator1
    Position Sampling=::NewEnumerator1
    SamplePosOffset=false
    Socket Index=&MergeHelper_IntSwitch(0=0, 1=1, 2=0, 3=1, 4=0, 5=1, Index=$Particles.PartitionParticles.Partition, Num=::NewEnumerator4)
    Socket Reader=$System.Socket Reader
    Socket Sampling Mode=::NewEnumerator3
    VelocitySpeedLimit=false
   )
   /SetVariables(Particles.RibbonTime:NiagaraFloat=&Add_Float(
     A=$Particles.RibbonTime
     B=&Multiply_Float(
      A=$Engine.DeltaTime
      B=&Multiply_Float(
       A=&FloatFromCurve(
        FloatCurve=DataInterface() {propertyValues {Curve {keys [{interpMode=RCIM_Cubic, leaveTangentWeight=0.3358, tangentMode=RCTM_User, time=0, value=-0.3}, <at 1 -1 RCIM_Cubic>]}}}
        CurveIndex=&NormalizeFloat(Value=$User.ParaGlideLeanAlpha, Min=0, Max=1)
       )
       B=1
      )
     )
    )) {isSetParameters=true}
   /SetVariables(Particles.InterpolatedTrailVector:Vector3f=&AddVectorToPosition(
     Position=Linked() {linkedVariable {name=Emitter.TargetVector, type=Vector3f}}
     Vector=&Lerp_Vector(
      A=<0 0 0>
      B=&MakeVector(
       X=&Add_Float(
        A=&Sine(
         Normalized Angle=&Frac_Float(Float=&Add_Float(
           A=&Add_Float(A=&Multiply_Float(A=$Emitter.Age, B=0), B=$Particles.RibbonTime)
           B=&Add_Float(
            A=&Multiply_Float(A=$Particles.Initial.PartitionParticles.Normalized Index In Partition, B=0.5)
            B=$Particles.PartitionParticles.NormalizedPartition
           )
          ))
         Period=1
         Scale=&FloatFromCurve(
          FloatCurve=DataInterface() {propertyValues {Curve {keys [<at 0 0>,<at 1 1>]}}}
          CurveIndex=$Particles.PartitionParticles.Normalized Index In Partition
          Scale Curve=0.5
         )
         Bias=0
        )
        B=&Multiply_Float(
         A=&SelectFloatFromArray(
          Direct Array Index=$Particles.PartitionParticles.Partition
          Direct Array Mode=::NewEnumerator1
          Float Selection Array=DataInterface() {propertyValues {FloatData [1,-1]}}
         )
         B=&Lerp_Float(A=0.3, B=0.3, Alpha=$User.ParaGlideLeanAlpha)
        )
       )
       Y=&Lerp_Float(
        A=0
        B=&Lerp_Float(A=-0.2, B=&FN_FrontEndFloatModifier(Gameplay=1, Front End=-0.2), Alpha=$User.ParaGlideLeanAlpha)
        Alpha=&FloatFromCurve(
         FloatCurve=DataInterface() {propertyValues {Curve {keys [{interpMode=RCIM_Cubic, tangentMode=RCTM_Break, time=0, value=0}, {interpMode=RCIM_Cubic, tangentMode=RCTM_Break, time=1, value=1}]}}}
         CurveIndex=$Particles.PartitionParticles.Normalized Index In Partition
        )
       )
       Z=0
      )
      Alpha=&FloatFromCurve(
       FloatCurve=DataInterface() {propertyValues {Curve {keys [
           {interpMode=RCIM_Cubic, tangentMode=RCTM_Break, time=0, value=0.5}
           <at 0.5 0.9 RCIM_Cubic>
           {arriveTangent=-1.5953, tangentMode=RCTM_User, time=1, value=0.5}
          ]}}}
       CurveIndex=$Particles.Initial.PartitionParticles.Normalized Index In Partition
      )
     )
    )) {isSetParameters=true}
   &InterpolateOverTime(
    Rate Of Change:NiagaraFloat=&Lerp_Float(A=2, B=12, Alpha=&NormalizeFloat(Value=$System.Dive, Min=0, Max=1))
    Target Position Value:NiagaraPosition=$Output.SocketLocation.SampledPosition
    Variable Type=::NewEnumerator6
   )
   /SetVariables(Particles.Interpolated_WPO:Vector4f=&MakeFloat4FromVectorAndFloat(
     "Vector (RGB)"=&TransformPosition(Position=$StackContext.InterpolateOverTime.InterpolateOverTime_Position.Moving Average, Source Space=::Local, Destination Space=::World)
     "Float (Alpha)"=1
    )) {isSetParameters=true}
   /SetVariables(Particles.InterpolatedPosition:NiagaraPosition=&AddVectorToPosition(
     Position=&Lerp_Vector(
      A=$Output.SocketLocation.SampledPosition
      B=Linked() {linkedVariable {name=StackContext.InterpolateOverTime.InterpolateOverTime_Position.Moving Average, type=NiagaraPosition}}
      Alpha=&FloatFromCurve(
       FloatCurve=DataInterface() {propertyValues {Curve {keys [
           {interpMode=RCIM_Cubic, leaveTangent=2.0815, tangentMode=RCTM_Break, time=0, value=0}
           {arriveTangent=0.1276, tangentMode=RCTM_User, time=1, value=1}
          ]}}}
       CurveIndex=$Particles.PartitionParticles.Normalized Index In Partition
      )
     )
     Vector=&Add_Vector(
      A=&Multiply_VectorByFloat(
       Vector=$Particles.InterpolatedTrailVector
       Float=&Multiply_Float(
        A=&Multiply_Float(A=$Emitter.RibbonLength, B=1)
        B=$Particles.PartitionParticles.Normalized Index In Partition
       )
      )
      B=<0 0 0>
     )
    )) {isSetParameters=true}
   /SetVariables(Particles.Position=$StackContext.InterpolatedPosition) {isSetParameters=true}
   /SetVariables(Particles.RibbonFacing=&Multiply_Vector(
     A=&SelectVectorFromArray(
      Direct Array Index=$Particles.PartitionParticles.Partition
      Direct Array Mode=::NewEnumerator1
      Vector Selection Array=DataInterface() {propertyValues {FloatData [<1 -1 0>, <1 1 0>]}}
     )
     B=<1 1 1>
    )) {isSetParameters=true}
   /SetVariables(
    Particles.RibbonWidth=&Multiply_Float(
     A=&FloatFromCurve(
      FloatCurve=DataInterface() {propertyValues {Curve {keys [
          {arriveTangent=4.6825, interpMode=RCIM_Cubic, leaveTangent=4.6825, tangentMode=RCTM_Break, time=0, value=0.3}
          {arriveTangent=1.0775, interpMode=RCIM_Cubic, leaveTangent=1.0775, tangentMode=RCTM_User, time=0.1643, value=0.7901}
          {arriveTangent=-0.3311, interpMode=RCIM_Cubic, leaveTangent=-0.3311, tangentMode=RCTM_User, time=0.5, value=1}
          <at 1 0.2 RCIM_Cubic>
         ]}}}
      CurveIndex=$Particles.PartitionParticles.Normalized Index In Partition
     )
     B=25
    )
    Particles.RibbonWidthGlow:NiagaraFloat=&Multiply_Float(
     A=&FloatFromCurve(
      FloatCurve=DataInterface() {propertyValues {Curve {keys [
          {arriveTangent=4.1315, interpMode=RCIM_Cubic, leaveTangent=4.1315, tangentMode=RCTM_User, time=0, value=0.5}
          {arriveTangent=0.8867, interpMode=RCIM_Cubic, leaveTangent=0.8867, tangentMode=RCTM_User, time=0.1, value=1}
          {arriveTangent=-0.4167, interpMode=RCIM_Cubic, leaveTangent=-0.4167, tangentMode=RCTM_User, time=0.5, value=1}
          {arriveTangent=-0.4217, interpMode=RCIM_Cubic, leaveTangent=-0.4217, tangentMode=RCTM_User, time=1, value=0.2}
         ]}}}
      CurveIndex=$Particles.PartitionParticles.Normalized Index In Partition
     )
     B=55
    )
   ) {isSetParameters=true}
   &DynamicMaterialParameters(
    Index 0 Param 1=$Particles.RibbonTime
    Module.Index0Param0Write=true
    Module.Index0Param1Write=false
    Module.Index0Param2Write=false
    Module.Index0Param3Write=false
    Module.Index2Param0Write=true
    Module.Index2Param1Write=false
    Module.Index2Param2Write=false
    Module.Index2Param3Write=false
    Write Parameter Index 2=true
   ) {module=DynamicMaterialParameters001}
   &ScaleColor(
    Scale Alpha=&Multiply_Float(A=2, B=&FNTimeOfDayScalar(Night Scale=1.3, Morning Scale=1, Evening Scale=1))
    Scale RGB=&MakeVectorFromLinearColorRGB(Color=&ColorFromCurve(
      ColorCurve=DataInterface() {propertyValues {
        AlphaCurve {keys [<at 0 1 RCIM_Cubic>,<at 1 1 RCIM_Cubic>]}
        BlueCurve {keys [<at 0.1147 0 RCIM_Cubic>,<at 0.3164 0.1017>,<at 0.5332 0.25 RCIM_Cubic>]}
        GreenCurve {keys [<at 0.1147 0.75 RCIM_Cubic>,<at 0.3164 0.4298>,<at 0.5332 1 RCIM_Cubic>]}
        RedCurve {keys [<at 0.1147 1 RCIM_Cubic>,<at 0.3164 1>,<at 0.5332 0 RCIM_Cubic>]}
       }}
      CurveIndex=$Particles.Initial.PartitionParticles.Normalized Index In Partition
     ))
    ScaleA=true
    ScaleRGB=true
   ) {module=ScaleColor003}
   &FN_Sprite_Attributes(
    Color=&MultiplyLinearColor(
     Linear Color=$StackContext.Color
     Scale Factor=&MakeVector4FromVectorAndFloat("Vector (RGB)"=&VectorFromFloat(Value=0.5), "Float (Alpha)"=0.5)
    )
    Create Color Attribute=true
    Write to Module Namespace=true
   )
   &FN_Sprite_Attributes(
    Color=&MultiplyLinearColor(
     Linear Color=&ColorFromCurve(
      ColorCurve=DataInterface() {propertyValues {
        AlphaCurve {keys [<at 0 1 RCIM_Cubic>,<at 0.3266 1>,<at 0.5306 1>,<at 0.7734 1 RCIM_Cubic>]}
        BlueCurve {keys [<at 0.3028 0 RCIM_Cubic>,<at 0.781 0.375 RCIM_Cubic>]}
        GreenCurve {keys [<at 0.3028 0.75 RCIM_Cubic>,<at 0.781 1.5 RCIM_Cubic>]}
        RedCurve {keys [<at 0.3028 1 RCIM_Cubic>,<at 0.781 0 RCIM_Cubic>]}
       }}
      CurveIndex=$Particles.Initial.PartitionParticles.Normalized Index In Partition
     )
     Scale Factor=&MakeVector4FromVectorAndFloat(
      "Vector (RGB)"=&VectorFromFloat(Value=1)
      "Float (Alpha)"=&Multiply_Float(
       A=&Multiply_Float(A=0.15, B=&FN_FrontEndFloatModifier(Gameplay=1, Front End=0.25))
       B=&FNTimeOfDayScalar(Night Scale=0.5, Morning Scale=0.7, Evening Scale=0.7)
      )
     )
    )
    Create Color Attribute=true
    Write to Module Namespace=true
   ) {module=FN_Sprite_Attributes001}
   &FN_Sprite_Attributes(
    Color=&MultiplyLinearColor(
     Linear Color=$StackContext.Color
     Scale Factor=&MakeVector4FromVectorAndFloat("Vector (RGB)"=&VectorFromFloat(Value=1), "Float (Alpha)"=0.2)
    )
    Create Color Attribute=true
    Write to Module Namespace=true
   ) {module=FN_Sprite_Attributes002}
   &DynamicMaterialParameters(
    Index 0 Param 1=$System.RevealCurveHands
    Module.Index0Float0Write=true
    Module.Index0Float1Write=false
    Module.Index0Float2Write=false
    Module.Index0Float3Write=false
   )
  ]
  rendererClasses [&NiagaraRibbonRendererProperties]
  renderers [
   {rendererClass=&NiagaraRibbonRendererProperties}
   {rendererClass=&NiagaraRibbonRendererProperties, rendererIndex=1}
   {rendererClass=&NiagaraRibbonRendererProperties, rendererIndex=2}
   {rendererClass=&NiagaraRibbonRendererProperties, rendererIndex=3}
  ]
 }
 Keys {
  emitterUpdateScript [
   &EmitterState(Life Cycle Mode=::NewEnumerator1, Loop Behavior=::NewEnumerator1, Loop Duration Mode=::NewEnumerator1, UseLoopDelay=false)
   &SpawnBurst_Instantaneous(Spawn Count=19, Use Spawn Probability=false, UseLoopCountLimit=true)
  ]
  particleSpawnScript [
   &InitializeParticle(
    Color=<rgba 0.938 0.6961 0 0.6>
    Lifetime=5
    Mesh Scale=<0.6 0.4 1>
    Mesh Scale Mode=::NewEnumerator4
    UsePositionOffset=false
    Write Material Random=true
   )
   &FN_SetSpawnGroup(Number of Spawn Groups=22)
   /SetVariables() {isSetParameters=true, module=SetVariables_#9}
   /SetVariables(Particles.GizmoScale:Vector3f=<5 5 5>) {isSetParameters=true, module=SetVariables_#10}
   /SetVariables(Particles.NormalizedParticleIndex:NiagaraFloat=&Power(a=&OneMinusFloat(Float=&ReturnNormalizedExecIndex()), b=3)) {isSetParameters=true, module=SetVariables_#11}
  ]
  particleUpdateScript [
   &ParticleState(Kill Particles When Lifetime Has Elapsed=false, Let Infinitely Lived Particles Die When Emitter Deactivates=true)
   /SetVariables(Particles.PreviousPos:NiagaraPosition=$Particles.Position) {isSetParameters=true, module=SetVariables_#12}
   &SkeletalMeshLocation(
    Bone Sampling Mode=::NewEnumerator4
    MaskSpawnGroup=false
    Orientation Sampling=::NewEnumerator1
    Sampled Position Offset=&Add_Vector(
     A=&MakeVector(
      X=&Add_Float(
       A=&FN_FrontEndFloatModifier(Gameplay=-70, Front End=-50)
       B=&Multiply_Float(A=&ReturnNormalizedExecIndex(), B=-260)
      )
      Y=&Add_Float(
       A=&Add_Float(A=50, B=0)
       B=&Lerp_Float(
        A=&Add_Float(
         A=&Add_Float(
          A=&Multiply_Float(
           A=&FloatFromCurve(
            FloatCurve=DataInterface() {propertyValues {Curve {
               defaultValue=1
               keys [
                {arriveTangent=-2.4782, interpMode=RCIM_Cubic, leaveTangent=-2.4782, leaveTangentWeight=0.1667, tangentMode=RCTM_User, time=0, value=1.3}
                {arriveTangentWeight=0.1667, interpMode=RCIM_Cubic, leaveTangentWeight=0.1667, tangentMode=RCTM_User, time=0.5, value=0}
                {arriveTangent=1.6086, tangentMode=RCTM_User, time=1, value=1}
               ]
              }}}
            CurveIndex=&ReturnNormalizedExecIndex()
           )
           B=-25
          )
          B=&Multiply_Float(
           A=&Sine(
            Normalized Angle=&Multiply_Float(
             A=&Add_Float(A=$System.Age, B=&ReturnNormalizedExecIndex(Normalized Index Scale=2))
             B=-1
            )
            Period=1
            Scale=1
            Bias=0
           )
           B=&FN_FrontEndFloatModifier(Gameplay=5, Front End=4)
          )
         )
         B=&FN_FrontEndFloatModifier(Gameplay=0, Front End=15)
        )
        B=&Add_Float(
         A=&Multiply_Float(
          A=&FloatFromCurve(
           FloatCurve=DataInterface() {propertyValues {Curve {
              defaultValue=1
              keys [
               {arriveTangent=-1.2393, leaveTangentWeight=0.1667, time=0, value=0}
               {
                arriveTangent=-0.9143
                arriveTangentWeight=0.1667
                interpMode=RCIM_Cubic
                leaveTangentWeight=0.1667
                tangentMode=RCTM_Break
                time=0.5
                value=0
               }
               {arriveTangent=3.1583, tangentMode=RCTM_User, time=1, value=1}
              ]
             }}}
           CurveIndex=&ReturnNormalizedExecIndex()
          )
          B=-50
         )
         B=&Multiply_Float(
          A=&FloatFromCurve(
           FloatCurve=DataInterface() {propertyValues {Curve {keys [
               <at 0 0 RCIM_Cubic>
               <at 0.2 1 RCIM_Cubic>
               {arriveTangentWeight=0.1795, interpMode=RCIM_Cubic, leaveTangentWeight=0.1795, tangentMode=RCTM_User, time=0.4, value=0}
               {arriveTangentWeight=0.0675, interpMode=RCIM_Cubic, leaveTangentWeight=0.0337, tangentMode=RCTM_User, time=0.6, value=0}
               <at 0.7 0.6 RCIM_Cubic>
               {interpMode=RCIM_Cubic, leaveTangent=6, time=0.85, value=0}
              ]}}}
           CurveIndex=&Frac_Float(Float=&Subtract_Float(
             A=&ReturnNormalizedExecIndex(Normalized Index Scale=-2, Normalized Index Includes One=false)
             B=&Multiply_Float(A=$Emitter.Age, B=-3)
            ))
          )
          B=8
         )
        )
        Alpha=$User.ParaGlideLeanAlpha
       )
      )
      Z=&Multiply_Float(
       A=$Particles.NormalizedParticleIndex
       B=&RemapRange(Input Value To Remap=$Particles.Roadbending, Input Min=-250, Input Max=250, Output Min=100, Output Max=-100)
      )
     )
     B=$System.MasterRoadOffset
    )
    Skeletal Mesh=$System.Skeletal Mesh
   )
   &self:SetRoadBending(CurrentPos:NiagaraPosition=$Particles.Position, MaxBending=250, RestSpeed=1, VelDivider=4) {module=SetRoadBending}
   /SetVariables(Particles.ZRotOffset:NiagaraFloat=&Multiply_Float(
     A=&FloatFromCurve(
      FloatCurve=DataInterface() {propertyValues {Curve {keys [
          {interpMode=RCIM_Cubic, leaveTangent=-1.084, tangentMode=RCTM_Break, time=0, value=1}
          {interpMode=RCIM_Cubic, tangentMode=RCTM_Break, time=1, value=0}
         ]}}}
      CurveIndex=&ReturnNormalizedExecIndex()
     )
     B=&RemapRange(Input Value To Remap=$Particles.Roadbending, Input Min=-250, Input Max=250, Output Min=40, Output Max=-40)
    )) {isSetParameters=true}
   /SetVariables(Particles.OrientationGoal:Quat4f=&MultiplyQuaternion(
     Quaternion A=&MultiplyQuaternion(
      Quaternion A=&MultiplyQuaternion(
       Quaternion A=$Particles.SkeletalMeshLocation.SampledRotation
       Quaternion B=&MakeQuatFromAxisAngle(Axis=<0 1 0>, AngleInDegrees=&Add_Float(A=90, B=0))
      )
      Quaternion B=&MakeQuatFromAxisAngle(Axis=<1 0 0>, AngleInDegrees=90)
     )
     Quaternion B=&MakeQuatFromAxisAngle(Axis=<0 0 1>, AngleInDegrees=$StackContext.ZRotOffset)
    )) {isSetParameters=true}
   /SetVariables(Particles.MeshOrientation=$Particles.OrientationGoal) {isSetParameters=true}
   &ScaleColor(
    Scale Alpha=&Multiply_Float(
     A=1
     B=&FloatFromCurve(
      FloatCurve=DataInterface() {propertyValues {Curve {keys [<at 0.3 1 RCIM_Cubic>,<at 1 0>]}}}
      CurveIndex=&ReturnNormalizedExecIndex()
     )
    )
    Scale RGB=&VectorFromFloat(Value=&RemapRange(
      Input Value To Remap=&Sine(
       Normalized Angle=&Add_Float(
        A=&ReturnNormalizedExecIndex(Normalized Index Scale=3)
        B=&Multiply_Float(
         A=&Add_Float(A=$Engine.System.Age, B=&Multiply_Float(A=$System.FN_Integrator002.Value, B=10))
         B=-1
        )
       )
       Period=1
       Scale=1
       Bias=0
      )
      Input Min=-1
      Input Max=1
      Output Min=&Lerp_Float(A=0.3, B=0.3, Alpha=$User.ParaGlideLeanAlpha)
      Output Max=1
     ))
    ScaleA=false
    ScaleRGB=true
   )
   &ScaleColor(
    Scale Alpha=&Multiply_Float(
     A=1
     B=&FloatFromCurve(
      FloatCurve=DataInterface() {propertyValues {Curve {keys [<at 0.6 1 RCIM_Cubic>,<at 1 0>]}}}
      CurveIndex=&ReturnNormalizedExecIndex()
      Scale Curve=&FN_FrontEndFloatModifier(Gameplay=1, Front End=0.25)
     )
    )
    Scale RGB=&VectorFromFloat(Value=&Multiply_Float(A=&Lerp_Float(A=1, B=7, Alpha=$Particles.NormalizedParticleIndex), B=1))
    ScaleA=true
    ScaleRGB=true
   ) {module=ScaleColor001}
   &FN_Sprite_Attributes(
    Color=&MultiplyLinearColor(
     Linear Color=<rgba 5 3.7105 0 1>
     Scale Factor=&MakeVector4FromVectorAndFloat(
      "Vector (RGB)"=<1 1 1>
      "Float (Alpha)"=&Multiply_Float(
       A=&Multiply_Float(
        A=0.1
        B=&FloatFromCurve(
         FloatCurve=DataInterface() {propertyValues {Curve {keys [<at 0.3 1 RCIM_Cubic>,<at 1 0>]}}}
         CurveIndex=&ReturnNormalizedExecIndex()
        )
       )
       B=&Multiply_Float(
        A=&FloatFromCurve(
         FloatCurve=DataInterface() {propertyValues {Curve {keys [{interpMode=RCIM_Cubic, tangentMode=RCTM_Break, time=0.5, value=0}, {interpMode=RCIM_Cubic, tangentMode=RCTM_Break, time=1.1, value=1}]}}}
         CurveIndex=&Subtract_Float(A=$System.Age, B=&ReturnNormalizedExecIndex(Normalized Index Scale=0.7))
        )
        B=&FN_FrontEndFloatModifier(Gameplay=1, Front End=0.5)
       )
      )
     )
    )
    Create Color Attribute=true
    Write to Module Namespace=true
   )
   &FN_Sprite_Attributes(
    Color=&MultiplyLinearColor(
     Linear Color=<rgba 0.05 0.0444 0 1>
     Scale Factor=&MakeVector4FromVectorAndFloat(
      "Vector (RGB)"=<1 1 1>
      "Float (Alpha)"=&Multiply_Float(
       A=&Multiply_Float(
        A=&FloatFromCurve(
         FloatCurve=DataInterface() {propertyValues {Curve {keys [{interpMode=RCIM_Cubic, tangentMode=RCTM_Break, time=0.5, value=0}, {interpMode=RCIM_Cubic, tangentMode=RCTM_Break, time=1.1, value=1}]}}}
         CurveIndex=&Subtract_Float(A=$System.Age, B=&ReturnNormalizedExecIndex(Normalized Index Scale=0.7))
        )
        B=0.7
       )
       B=&FloatFromCurve(
        FloatCurve=DataInterface() {propertyValues {Curve {keys [<at 0.3 1 RCIM_Cubic>,<at 1 0>]}}}
        CurveIndex=&ReturnNormalizedExecIndex()
        Scale Curve=0.6
       )
      )
     )
    )
    Create Color Attribute=true
    Write to Module Namespace=true
   ) {module=FN_Sprite_Attributes001}
   &FN_Sprite_Attributes(
    Color=&MultiplyLinearColor(
     Linear Color=<rgba 0.05 0.0317 0 1>
     Scale Factor=&MakeVector4FromVectorAndFloat(
      "Vector (RGB)"=<1 1 1>
      "Float (Alpha)"=&Multiply_Float(
       A=&Multiply_Float(
        A=&FloatFromCurve(
         FloatCurve=DataInterface() {propertyValues {Curve {keys [{interpMode=RCIM_Cubic, tangentMode=RCTM_Break, time=0.5, value=0}, {interpMode=RCIM_Cubic, tangentMode=RCTM_Break, time=1.1, value=1}]}}}
         CurveIndex=&Subtract_Float(A=$System.Age, B=&ReturnNormalizedExecIndex(Normalized Index Scale=0.7))
        )
        B=1
       )
       B=&FloatFromCurve(
        FloatCurve=DataInterface() {propertyValues {Curve {keys [<at 0.3 1 RCIM_Cubic>,<at 1 0>]}}}
        CurveIndex=&ReturnNormalizedExecIndex()
        Scale Curve=0.4
       )
      )
     )
    )
    Create Color Attribute=true
    Write to Module Namespace=true
   ) {module=FN_Sprite_Attributes002}
   /SetVariables(Particles.DP0Glow:Vector4f=<1 0 0 0>, Particles.DP0Glow2:Vector4f=<4 0 0 0>) {isSetParameters=true}
   &DynamicMaterialParameters(
    Index 0 Param 1=&ReturnNormalizedExecIndex()
    Module.Index0Float0Write=true
    Module.Index0Float1Write=true
    Module.Index0Float2Write=true
    Module.Index0Float3Write=true
   )
   /SetVariables(
    Particles.SmallKeyColor:LinearColor=&MultiplyLinearColor(
     Linear Color=<rgba 0.938 0.6961 0 1>
     Scale Factor=&MakeVector4FromVectorAndFloat(
      "Vector (RGB)"=&Lerp_Vector(
       A=&VectorFromFloat(Value=&Multiply_Float(A=5, B=1))
       B=&VectorFromFloat(Value=&RemapRange(
         Input Value To Remap=&Sine(
          Normalized Angle=&Add_Float(
           A=&ReturnNormalizedExecIndex(Normalized Index Scale=3)
           B=&Multiply_Float(
            A=&Add_Float(A=$Engine.System.Age, B=&Multiply_Float(A=$System.FN_Integrator002.Value, B=10))
            B=-1
           )
          )
          Period=1
          Scale=1
          Bias=0
         )
         Input Min=-1
         Input Max=1
         Output Min=0.5
         Output Max=5
        ))
       Alpha=$User.ParaGlideLeanAlpha
      )
      "Float (Alpha)"=&Multiply_Float(
       A=&Multiply_Float(
        A=&FloatFromCurve(
         FloatCurve=DataInterface() {propertyValues {Curve {keys [<at 0.3 1 RCIM_Cubic>,<at 1 0>]}}}
         CurveIndex=&ReturnNormalizedExecIndex()
        )
        B=&FloatFromCurve(
         FloatCurve=DataInterface() {propertyValues {Curve {keys [<at 0.3 1 RCIM_Cubic>,<at 1 0>]}}}
         CurveIndex=&ReturnNormalizedExecIndex()
        )
       )
       B=&FloatFromCurve(
        FloatCurve=DataInterface() {propertyValues {Curve {keys [{interpMode=RCIM_Cubic, tangentMode=RCTM_Break, time=0.6, value=0}, {interpMode=RCIM_Cubic, tangentMode=RCTM_Break, time=1.2, value=1}]}}}
        CurveIndex=&Subtract_Float(A=$System.Age, B=&ReturnNormalizedExecIndex(Normalized Index Scale=0.7))
        Scale Curve=&FN_FrontEndFloatModifier(Gameplay=1, Front End=0.25)
       )
      )
     )
    )
    Particles.SmallKeyPosition:NiagaraPosition=&AddVectorToPosition(Position=$StackContext.Position, Vector=&TransformVector(Vector=<0 0 0>, Destination Space=::World))
    Particles.SmallKeyScale:Vector3f=&Multiply_VectorByFloat(
     Vector=<0.28 0.3 0.5>
     Float=&Multiply_Float(
      A=&SelectFloatFromArray(
       Direct Array Index=$Particles.SpawnGroup
       Direct Array Mode=::NewEnumerator1
       Float Selection Array=DataInterface() {propertyValues {FloatData [1,1,0,1,1,1,0]}}
      )
      B=&ClampFloat(Float=&Multiply_Float(
        A=&FloatFromCurve(
         FloatCurve=DataInterface() {propertyValues {Curve {keys [
             {interpMode=RCIM_Cubic, leaveTangent=0.861, tangentMode=RCTM_Break, time=0.6, value=0.5}
             {interpMode=RCIM_Cubic, tangentMode=RCTM_Break, time=1.2, value=1}
            ]}}}
         CurveIndex=&Subtract_Float(A=$System.Age, B=&ReturnNormalizedExecIndex(Normalized Index Scale=0.7))
        )
        B=1
       ))
     )
    )
   ) {isSetParameters=true}
   &ScaleColor(
    Scale Alpha=&Multiply_Float(
     A=&FloatFromCurve(
      FloatCurve=DataInterface() {propertyValues {Curve {keys [{interpMode=RCIM_Cubic, tangentMode=RCTM_Break, time=0.5, value=0}, {interpMode=RCIM_Cubic, tangentMode=RCTM_Break, time=1.1, value=1}]}}}
      CurveIndex=&Subtract_Float(A=$System.Age, B=&ReturnNormalizedExecIndex(Normalized Index Scale=0.7))
     )
     B=1
    )
    Scale RGB=&VectorFromFloat(Value=&Multiply_Float(
      A=&FloatFromCurve(
       FloatCurve=DataInterface() {propertyValues {Curve {keys [{interpMode=RCIM_Cubic, tangentMode=RCTM_Break, time=0.6, value=10}, <at 0.8 1 RCIM_Cubic>]}}}
       CurveIndex=&Subtract_Float(A=$System.Age, B=&ReturnNormalizedExecIndex(Normalized Index Scale=0.7))
      )
      B=1
     ))
    ScaleA=true
    ScaleRGB=true
   ) {module=ScaleColor002}
   &ScaleMeshSize(Scale Factor=&VectorFromFloat(Value=&ClampFloat(Float=&Multiply_Float(
       A=&FloatFromCurve(
        FloatCurve=DataInterface() {propertyValues {Curve {keys [
            {interpMode=RCIM_Cubic, leaveTangent=0.861, tangentMode=RCTM_Break, time=0.5, value=0.5}
            {interpMode=RCIM_Cubic, tangentMode=RCTM_Break, time=1.1, value=1}
           ]}}}
        CurveIndex=&Subtract_Float(A=$System.Age, B=&ReturnNormalizedExecIndex(Normalized Index Scale=0.7))
       )
       B=1
      ))))
  ]
  rendererClasses [&NiagaraMeshRendererProperties]
  renderers [
   {rendererClass=&NiagaraMeshRendererProperties, rendererIndex=0}
   {rendererClass=&NiagaraMeshRendererProperties, rendererIndex=1}
   {rendererClass=&NiagaraMeshRendererProperties, rendererIndex=2}
   {rendererClass=&NiagaraMeshRendererProperties, rendererIndex=3}
   {rendererClass=&NiagaraMeshRendererProperties, rendererIndex=4}
   {rendererClass=&NiagaraMeshRendererProperties, rendererIndex=5}
  ]
 }
 NotesBottomTrails {
  emitterUpdateScript [
   &EmitterState(Inactive Response=::NewEnumerator1, Life Cycle Mode=::NewEnumerator1, Loop Delay=0.2, Loop Duration=1, UseLoopDelay=false)
   &SpawnRate(
    SpawnRate=&Multiply_Float(
     A=12
     B=&Lerp_Float(A=1, B=&FN_FrontEndFloatModifier(Gameplay=4, Front End=2), Alpha=$User.ParaGlideLeanAlpha)
    )
    Use Spawn Probability=false
   )
  ]
  particleSpawnScript [
   /SetVariables(Particles.RandomPerParticle:NiagaraFloat=&UniformAOrBFloat(A=0, Evaluation Type=::NewEnumerator0, Recalculate Random Each Loop=true, Override Seed=false)) {isSetParameters=true, module=SetVariables_#13}
   &FN_SetSpawnGroup(Number of Spawn Groups=2)
   &InitializeParticle(
    Color=<rgba 2 1.4945 0.04 1>
    Lifetime=&Multiply_Float(
     A=&RandomRangeFloat(Minimum=5, Maximum=4)
     B=&Lerp_Float(A=1, B=0.3, Alpha=$User.ParaGlideLeanAlpha)
    )
    Sprite Rotation Angle=&RandomRangeFloat(Minimum=-45, Maximum=45)
    Sprite Rotation Mode=::NewEnumerator0
    Sprite Size Mode=::NewEnumerator3
    Uniform Sprite Size=&Multiply_Float(
     A=&RandomRangeFloat(Minimum=20, Maximum=40)
     B=&Lerp_Float(A=1.3, B=0.3, Alpha=$Particles.RandomPerParticle)
    )
    UsePositionOffset=false
    Write Material Random=true
   ) {module=InitializeParticle001}
   &DynamicMaterialParameters(
    Index 0 Param 1=&UniformAOrBFloat(A=0.333, Evaluation Type=::NewEnumerator0, Recalculate Random Each Loop=true, Override Seed=false)
    Module.Index0Param0Write=true
    Module.Index0Param1Write=false
    Module.Index0Param2Write=false
    Module.Index0Param3Write=false
   )
   /SetVariables(
    Particles.SubImageIndexNotes:NiagaraFloat=&RandomRangeFloat(Minimum=0, Maximum=3)
    Particles.VisibilityTag=&MakeIntFromFloat(Float=$Particles.RandomPerParticle)
   ) {isSetParameters=true, module=SetVariables_#14}
  ]
  particleUpdateScript [
   &ParticleState() {module=ParticleState001}
   &self:FollowRibbon(
    Alpha:NiagaraFloat=&FloatFromCurve(FloatCurve=DataInterface() {propertyValues {Curve {keys [<at 0 0>,<at 1 1>]}}})
    ParticlePerRibbon=30
    RibbonNumber:NiagaraInt32=&SpawnGroup()
    RibbonReader=DataInterface() {propertyValues {EmitterBinding {bindingMode=Other, emitterName=BottomTrails}}}
   ) {module=FollowRibbon}
   /SetVariables(Particles.Position=&AddVectorToPosition(
     Position=&Multiply_Vector(
      A=Linked() {linkedVariable {name=Output.FollowRibbon.RibbonPos, type=NiagaraPosition}}
      B=&VectorFromFloat(Value=1)
     )
     Vector=&Multiply_VectorByFloat(
      Vector=&RandomRangeVector(Minimum=<-1 -1 -1>, Maximum=<1 1 1>)
      Float=&Multiply_Float(A=&FloatFromCurve(FloatCurve=DataInterface() {propertyValues {Curve {keys [<at 0 0>,<at 1 1>]}}}), B=50)
     )
    )) {isSetParameters=true, module=SetVariables_#15}
   &ParticleState()
   &SpriteRotationRate(Rotation Rate=&Lerp_Float(
     A=&UniformAOrBFloat(A=-20, B=20, Evaluation Type=::NewEnumerator0, Recalculate Random Each Loop=true, Override Seed=false)
     B=&UniformAOrBFloat(A=-100, B=100, Evaluation Type=::NewEnumerator0, Recalculate Random Each Loop=true, Override Seed=false)
     Alpha=$Particles.RandomPerParticle
    ))
   &CameraOffset(Camera Offset Amount=5)
   &ScaleColor(Scale Alpha=$System.FadeInCurve, ScaleA=true, ScaleRGB=false)
   /SetVariables(Particles.SpriteSize=&Multiply_Vector2D_ByFloat(
     Float=&Multiply_Float(
      A=&FloatFromCurve(FloatCurve=DataInterface() {propertyValues {Curve {keys [<at 0 0.5>,<at 0.7 1>]}}})
      B=&UniformAOrBFloat(
       A=&FloatFromCurve(FloatCurve=DataInterface() {propertyValues {Curve {keys [
            <at 0 0 RCIM_Cubic>
            {arriveTangent=9.6335, interpMode=RCIM_Cubic, leaveTangent=2.7827, tangentMode=RCTM_Break, time=0.1, value=1}
            {arriveTangent=1.0943, interpMode=RCIM_Cubic, leaveTangent=1.0943, time=0.3322, value=0.4478}
            {arriveTangent=4.272, interpMode=RCIM_Cubic, leaveTangent=0.772, tangentMode=RCTM_Break, time=0.5569, value=1.5}
            {arriveTangent=-1.4573, interpMode=RCIM_Cubic, leaveTangent=-1.4573, time=0.6943, value=0.7273}
            {arriveTangent=2.7202, interpMode=RCIM_Cubic, leaveTangent=-2.3793, tangentMode=RCTM_Break, time=0.9, value=1}
            <at 0.95 0 RCIM_Cubic>
           ]}}})
       B=&FloatFromCurve(FloatCurve=DataInterface() {propertyValues {Curve {keys [
            <at 0 0 RCIM_Cubic>
            {arriveTangent=9.6335, interpMode=RCIM_Cubic, leaveTangent=2.7827, tangentMode=RCTM_Break, time=0.1, value=0.9794}
            {arriveTangent=0.0257, interpMode=RCIM_Cubic, leaveTangent=0.0257, time=0.5397, value=0.5}
            {arriveTangent=2.7202, interpMode=RCIM_Cubic, leaveTangent=-2.3793, tangentMode=RCTM_Break, time=0.9, value=1}
            <at 0.95 0 RCIM_Cubic>
           ]}}})
       Evaluation Type=::NewEnumerator0
       Recalculate Random Each Loop=true
       Override Seed=false
      )
     )
     Vector2D=$Particles.Initial.SpriteSize
    )) {isSetParameters=true, module=SetVariables_#16}
   &FN_Sprite_Attributes(
    Color=&MultiplyLinearColor(
     Linear Color=<rgba 2 0.7333 0.1 1>
     Scale Factor=&MakeVector4FromVectorAndFloat("Vector (RGB)"=<1 1 1>, "Float (Alpha)"=$System.FadeInCurve)
    )
    Create Color Attribute=true
    Write to Module Namespace=true
   ) {module=FN_Sprite_Attributes001}
   &FN_Sprite_Attributes(
    Color=&MultiplyLinearColor(
     Linear Color=<rgba 0.087 0.0181 0 1>
     Scale Factor=&MakeVector4FromVectorAndFloat("Vector (RGB)"=<1 1 1>, "Float (Alpha)"=$System.FadeInCurve)
    )
    Create Color Attribute=true
    Write to Module Namespace=true
   )
  ]
  rendererClasses [&NiagaraSpriteRendererProperties]
  renderers [
   {rendererClass=&NiagaraSpriteRendererProperties}
   {rendererClass=&NiagaraSpriteRendererProperties, rendererIndex=1}
   {rendererClass=&NiagaraSpriteRendererProperties, rendererIndex=2}
   {rendererClass=&NiagaraSpriteRendererProperties, rendererIndex=3}
   {rendererClass=&NiagaraSpriteRendererProperties, rendererIndex=4}
   {rendererClass=&NiagaraSpriteRendererProperties, rendererIndex=5}
  ]
 }
}
summary {
 systemName=NS_Contrail_DryEraseToro
 emitters [
  {emitterName=Keys, rendererClasses [&NiagaraMeshRendererProperties]}
  {emitterName=BottomTrails, rendererClasses [&NiagaraRibbonRendererProperties, &NiagaraMeshRendererProperties]}
  {emitterName=HandNotes, rendererClasses [&NiagaraSpriteRendererProperties]}
  {emitterName=NotesBottomTrails, rendererClasses [&NiagaraSpriteRendererProperties]}
  {emitterName=HandRibbons, rendererClasses [&NiagaraRibbonRendererProperties]}
  {bEnabled=false, emitterName=FN_Debug, rendererClasses [&NiagaraSpriteRendererProperties]}
  {emitterName=Decal_Glow001, rendererClasses [&NiagaraSpriteRendererProperties]}
 ]
 userVariables [
  User.bIsFrontEnd {type=NiagaraBool}
  User.bIsFrontEndPreview {type=NiagaraBool}
  User.ParaGlideLeanAlpha {type=NiagaraFloat, defaultValue=0, description=""}
  User.PlayerSpeed {type=NiagaraFloat, defaultValue=0, description=""}
  User.RibbonIDOffset {type=NiagaraInt32}
 ]
}
